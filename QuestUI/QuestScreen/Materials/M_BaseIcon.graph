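# M_BaseIcon is a quest screen UI material that draws a tintable icon from a texture parameter, defaulting to the Battle Pass quest icon.
# Only the texture's alpha is used, as a mask multiplied by the Color parameter and vertex alpha to give premultiplied color.
# A separate product of that alpha and Additive Ratio suits the alpha-composite blend, where lower opacity shifts the icon toward additive.
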Material /QuestUI/QuestScreen/Materials/M_BaseIcon
graph {
 properties {blendMode=BLEND_AlphaComposite, materialDomain=MD_UI, pixelDepthOffsetMode=PDOM_Legacy, usedShadingModels=Default Lit}
 nodes [
  n1 = TextureSample(Tex=TextureObjectParameter() {parameterName=Texture, texture=/QuestUI/QuestScreen/Textures/Icon_Battle_Pass_Quest}) {texture=&Icon_Battle_Pass_Quest}
  Multiply(A=@n1.A, B=ScalarParameter() {defaultValue=1, parameterName=Additive Ratio}) {constB=0.5}
  Multiply(A=VertexColor().A, B=Multiply(A=VectorParameter() {parameterName=Color, defaultValue=<rgba 0.75 0.75 0.75 1>}.RGB, B=@n1.A))
 ]
}
