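# M_UI_Empty is a blank User Interface domain material in the QuestScreen materials folder of QuestUI.
# It uses translucent blending and has no expression nodes, so no graph logic drives its output.
# Automatic setting of usage flags in the editor is turned off.
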Material /QuestUI/QuestScreen/Materials/M_UI_Empty
graph {
 expressions {}
 properties {bAutomaticallySetUsageInEditor=false, blendMode=BLEND_Translucent, materialDomain=MD_UI, usedShadingModels=Default Lit}
}
