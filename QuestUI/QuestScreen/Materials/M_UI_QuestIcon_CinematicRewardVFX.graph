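# Alpha-composite UI material for the cinematic reward quest icon, layering animated glow effects with the icon texture.
# A time-rotated angle gradient feeds a cosine color gradient, shaped by SDF box falloffs.
# Named reroutes carry polar-panned noise, two rotating triangle-texture particle layers and a rotating box-outline sweep.
# IsLockedCrossfade and IsSeenCrossfade are summed and capped to blend icon color and mask away from the animated look.
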Material /QuestUI/QuestScreen/Materials/M_UI_QuestIcon_CinematicRewardVFX
graph {
 properties {bAutomaticallySetUsageInEditor=false, blendMode=BLEND_AlphaComposite, materialDomain=MD_UI, usedShadingModels=Default Lit}
 nodes [
  n1 = NamedRerouteUsage()
  n2 = NamedRerouteUsage()
  n3 = Multiply(A=Time()) {constB=0.2}
  n4 = MaterialFunctionCall!(UVs, Rotation=OneMinus(None=Frac(None=@n3))) {materialFunction=/FortUILibrary/MaterialFunctions/Gradients/MF_UI_AngleGradient}
  n5 = MaterialFunctionCall(R=Constant() {r=0.025}, G=Constant() {r=0.1}, B=Constant() {r=1.1}) {materialFunction=&AppendMany}
  n6 = MaterialFunctionCall(R=Constant(), G=Constant(), B=Constant() {r=0.5}) {materialFunction=&AppendMany}
  n7 = MaterialFunctionCall(Size=NamedRerouteUsage()) {materialFunction=&MF_UI_Box}
  n8 = LinearInterpolate(A=NamedRerouteUsage(), Alpha=SmoothStep(Max=Constant() {r=1.5}, Value=OneMinus(None=@n7.SDFOutline)) {constMin=0.5})
  n9 = Multiply(
   A=Multiply(
    A=LinearInterpolate(A=@n1, Alpha=@n2)
    B=MaterialFunctionCall!(Gradient=@n4.Result, a=@n5.RGB, b=@n6.RGB, c, d) {materialFunction=/FortUILibrary/MaterialFunctions/Color/MF_UI_CosineGradient}.Result
   )
   B=Saturate(None=MaterialFunctionCall("Value (0-1)"=@n8) {materialFunction=&MF_UI_EaseCurves}.EaseInCubic)
  )
  n10 = Desaturation(None=@n9)
  n11 = NamedRerouteUsage()
  n12 = Min(A=Add(A=ScalarParameter() {defaultValue=1, parameterName=IsLockedCrossfade}, B=ScalarParameter() {parameterName=IsSeenCrossfade}))
  NamedRerouteDeclaration(None=LinearInterpolate(A=Add(A=@n10, B=@n11), B=NamedRerouteUsage(), Alpha=@n12)) {name=IconMask, nodeColor=<rgba 0 0 0 1>}
  n13 = MaterialFunctionCall(Float2=MaterialFunctionCall!(Vector or UVs=MaterialFunctionCall() {materialFunction=&GetUserInterfaceUV}.Normalized UV) {materialFunction=/FortUILibrary/MaterialFunctions/Basics/MF_UI_CheaperVectorToRadial}.Radial Coordinates) {materialFunction=&BreakOutFloat2Components}
  n14 = Multiply(A=@n13.R, B=ScalarParameter() {defaultValue=1, parameterName=PerpendicularRepitition})
  n15 = Multiply(A=@n13.G, B=ScalarParameter() {defaultValue=2, parameterName=ParallelRepitition})
  n16 = ScalarParameter() {defaultValue=-0.05, parameterName=PerpendicularSpeed}
  n17 = ScalarParameter() {defaultValue=-0.1, parameterName=ParallelSpeed}
  n18 = MaterialFunctionCall() {materialFunction=&GetUserInterfaceUV}
  n19 = TextureSample!(
   UVs=Panner(Coordinate=AppendVector(A=@n14, B=@n15), Time=Time(), Speed=AppendVector(A=@n16, B=@n17))
   Tex
   "DDX(UVs)"=DDX(Value=@n18.Normalized UV)
   "DDY(UVs)"=DDY(Value=@n18.Normalized UV)
   Apply View MipBias
  ) {mipValueMode=TMVM_Derivative, samplerType=SAMPLERTYPE_Grayscale, texture=/FortUILibrary/UIKit/Buttons/Textures/LowResBlurredNoise}
  n20 = Max(A=NamedRerouteUsage()) {constB=0.0005}
  NamedRerouteDeclaration(None=Multiply(A=@n19.R, B=Divide(B=@n20))) {name=RadialNoise, nodeColor=<rgba 0 0 0 1>}
  n21 = OneMinus(None=NamedRerouteUsage())
  n22 = TextureCoordinate()
  n23 = Time()
  n24 = MaterialFunctionCall(UVs=Rotator(Coordinate=@n22, Time=@n23) {speed=-0.1}, Texture Scale=ScalarParameter() {defaultValue=0.2, parameterName=Dots_Scale}) {materialFunction=&ScaleUVsByCenter}
  n25 = AppendVector(A=ScalarParameter() {defaultValue=-0.005, parameterName=Dots_U_Speed}, B=ScalarParameter() {defaultValue=0.0375, parameterName=Dots_V_Speed})
  n26 = MaterialFunctionCall(UVs=Rotator(Coordinate=@n22, Time=Time()) {speed=-0.25}, Texture Scale=ScalarParameter() {defaultValue=0.5, parameterName=Dots2_Scale}) {materialFunction=&ScaleUVsByCenter}
  n27 = Time()
  n28 = AppendVector(
   A=ScalarParameter() {defaultValue=0.005, parameterName=Dots2_U_Speed}
   B=ScalarParameter() {defaultValue=0.0125, parameterName=Dots2_V_Speed}
  )
  n29 = Add(
   A=TextureSample!(UVs=Panner(Coordinate=@n24.UVs, Time=@n23, Speed=@n25), Tex, Apply View MipBias) {samplerType=SAMPLERTYPE_LinearColor, texture=/Game/UI/Foundation/Items/Textures/T-Card-Hero-Anim-Triangle}.G
   B=TextureSample(UVs=Panner(Coordinate=@n26.UVs, Time=@n27, Speed=@n28)) {samplerType=SAMPLERTYPE_LinearColor, texture=&T-Card-Hero-Anim-Triangle}.R
  )
  n30 = Multiply(A=SmoothStep(Min=Power(Base=@n21) {constExponent=4}, Value=Saturate(None=@n29)), B=NamedRerouteUsage())
  NamedRerouteDeclaration(None=Multiply(A=Saturate(None=@n30), B=ScalarParameter() {defaultValue=1, parameterName=DotsAmount})) {name=Particles, nodeColor=<rgba 0 0 0 1>}
  n31 = Multiply(A=Time(), B=ScalarParameter() {defaultValue=2, parameterName=RotationSpeed}) {constB=-1}
  n32 = MaterialFunctionCall(Vector or UVs=Rotator(
    Coordinate=MaterialFunctionCall() {materialFunction=&GetUserInterfaceUV}.Normalized UV
    Time=Multiply(A=@n31) {constB=-1}
   ) {speed=1}) {materialFunction=&MF_UI_CheaperVectorToRadial}
  n33 = Multiply(A=@n32.Vector Converted to Angle) {constB=16}
  n34 = LinearInterpolate(B=Saturate(None=Power(Base=@n33)), Alpha=SmoothStep(Value=@n32.Vector Converted to Angle) {constMax=0, constMin=0.95})
  n35 = MaterialFunctionCall(
   UVs=MaterialFunctionCall() {materialFunction=&GetUserInterfaceUV}.Normalized UV
   Size=AppendVector(A=Constant() {r=0.68}, B=Constant() {r=0.5})
   CornerRoundness=Constant() {r=0.2}
   GlowMin=Multiply(A=OneMinus(None=@n32.Vector Converted to Angle), B=ScalarParameter() {defaultValue=-0.1, parameterName=GlowMin})
   GlowMax=Multiply(A=@n32.Vector Converted to Angle, B=ScalarParameter() {defaultValue=0.333, parameterName=GlowMax})
  ) {materialFunction=&MF_UI_Box}
  n36 = Saturate(None=@n35.Outline)
  NamedRerouteDeclaration(None=Saturate(None=DotProduct(A=@n34, B=@n36))) {name=RadialAnimation, nodeColor=<rgba 0 0 0 1>}
  NamedRerouteDeclaration(None=@n35.SDFOutline) {name=BoxOutlineSDF, nodeColor=<rgba 0 0 0 1>}
  n37 = NamedRerouteUsage()
  n38 = NamedRerouteUsage()
  n39 = ComponentMask(None=NamedRerouteUsage()) {r=true}
  n40 = TextureSample(UVs=MaterialFunctionCall(Texture Scale=Add(A=@n39) {constB=0.12, desc=Add to account for image padding}) {materialFunction=&ScaleUVsByCenter}.UVs) {texture=/QuestUI/QuestScreen/Textures/T_UI_QuestIcon_CinematicReward_S}
  n41 = NamedRerouteDeclaration(None=@n40.RGB) {name=TextureRGB, nodeColor=<rgba 0 0.41 1 1>}
  n42 = NamedRerouteUsage()
  NamedRerouteDeclaration(None=LinearInterpolate(A=LinearInterpolate(A=@n9, B=@n41, Alpha=@n42), B=Multiply(A=@n37, B=@n38), Alpha=@n12)) {name=IconColor, nodeColor=<rgba 0 0 0 1>}
  NamedRerouteDeclaration(None=@n40.A) {name=TextureAlpha, nodeColor=<rgba 0 0.41 1 1>}
  n43 = AppendVector(A=ScalarParameter() {defaultValue=0.68, parameterName=GlowWidth}, B=ScalarParameter() {defaultValue=0.5, parameterName=GlowHeight})
  NamedRerouteDeclaration(None=MaterialFunctionCall(
    Size=NamedRerouteDeclaration(None=@n43) {name=GlowSize, nodeColor=<rgba 1 0.5839 0.0355 1>}
    Corner Radius=ScalarParameter() {defaultValue=10, parameterName=CornerRadius}
    Radius in Pixels=Constant() {r=1}
   ) {materialFunction=&MF_UI_CoreUI_RoundedMask}.Outline SDF) {name=OutlineSDF, nodeColor=<rgba 0 0 0 1>}
  NamedRerouteUsage()
  NamedRerouteUsage()
 ]
}
