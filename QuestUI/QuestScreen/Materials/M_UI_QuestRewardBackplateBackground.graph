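# Translucent UI material for the quest reward backplate background, with a rounded shape and button-state styling.
# Chained base and selectable state lerp functions supply color and opacity; base Color feeds both Base States inputs, including opacity.
# That color multiplies the Slate post buffer 0 sample, then blends to fixed gray along the outline.
# A rounded mask with per-corner radius parameters supplies the outline and multiplies state opacity for the alpha.
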Material /QuestUI/QuestScreen/Materials/M_UI_QuestRewardBackplateBackground
graph {
 properties {bAutomaticallySetUsageInEditor=false, blendMode=BLEND_Translucent, materialDomain=MD_UI, usedShadingModels=Default Lit}
 nodes [
  n1 = MaterialFunctionCall!(Idle, Idle Opacity, Hovered, Hovered Opacity, Focused, Focused Opacity, Pressed, Pressed Opacity, Disabled, Disabled Opacity) {materialFunction=/FortUILibrary/Systems/ModularButton/MaterialFunctions/MF_UIKit_ButtonStatesBase_Lerp}
  n2 = MaterialFunctionCall!(
   Base States=@n1.Color
   Base States Opacity=@n1.Color
   Selected_Idle
   Selected_Idle Opacity
   Selected_Hovered
   Selected_Hovered Opacity
   Selected_Focused
   Selected_Focused Opacity
   Selected_Pressed
   Selected_Pressed Opacity
   Selected_Disabled
   Selected_Disabled Opacity
  ) {materialFunction=/FortUILibrary/Systems/ModularButton/MaterialFunctions/MF_UIKit_ButtonStatesSelectable_Lerp}
  n3 = VectorParameter() {
   group=Shape
   parameterName=Corner Radius
   sortPriority=1
   channelNames {a=BottomRight, b=TopRight, g=BottomLeft, r=TopLeft}
   defaultValue=<rgba 16 16 16 16>
  }
  n4 = MaterialFunctionCall(
   Corner Radius=AppendVector(A=@n3, B=@n3)
   Radius in Pixels=ScalarParameter() {defaultValue=1, group=Shape, parameterName=Radius in Pixels, sliderMax=1, sortPriority=2}
   Outline Thickness=Constant() {r=2}
  ) {materialFunction=&MF_UI_CoreUI_RoundedMask}
  LinearInterpolate(
   A=Multiply(A=MaterialFunctionCall!("UVs (Optional)") {materialFunction=/Engine/Functions/UserInterface/GetSlatePost0}.LinearRGB, B=@n2.Color)
   B=Constant3Vector() {constant=<rgba 0.2195 0.2195 0.2195 1>}
   Alpha=@n4.Outline Mask
  )
  Multiply(A=@n4.Shape Mask, B=@n2.Opacity)
 ]
}
