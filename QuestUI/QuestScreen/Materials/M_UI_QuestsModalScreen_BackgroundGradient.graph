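# Translucent UI material for the background gradient of the quests modal screen.
# MF_UI_LinearGradient builds a linear gradient from parameters for start, end, rotation angle, rotation center and offset (exposed as HorizontalOffset).
# The gradient is multiplied by an Opacity scalar, alongside a separate Color vector parameter for the tint.
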Material /QuestUI/QuestScreen/Materials/M_UI_QuestsModalScreen_BackgroundGradient
graph {
 properties {bAutomaticallySetUsageInEditor=false, blendMode=BLEND_Translucent, materialDomain=MD_UI, usedShadingModels=Default Lit}
 nodes [
  Multiply(
   A=MaterialFunctionCall(
    GradientStart=ScalarParameter() {parameterName=GradientStart}
    GradientEnd=ScalarParameter() {defaultValue=1, parameterName=GradientEnd}
    RotationAngle=ScalarParameter() {defaultValue=0.5, parameterName=RotationAngle}
    RotationCenter=ScalarParameter() {defaultValue=0.5, parameterName=RotationCenter}
    GradientOffset=ScalarParameter() {parameterName=HorizontalOffset}
   ) {materialFunction=&MF_UI_LinearGradient}.Result
   B=ScalarParameter() {defaultValue=1, parameterName=Opacity}
  )
  VectorParameter() {parameterName=Color, defaultValue=<rgba 0.0736 0 1 1>}
 ]
}
