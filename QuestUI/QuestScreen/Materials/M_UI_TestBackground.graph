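# UI-domain material for a quest screen test background, building a procedural colored ellipse gradient.
# Widget pixel size corrects the UV aspect ratio for exponential radial gradients that blend two colors, with AnimElipseTransition fading the ellipse.
# A linear gradient masks an alpha-weighted FadeColor tint before MF_UI_Gradient_TransitionSaturation produces the Iteration2BG reroute.
# A separate offset, scaled radial mask drives a blend gated by IsShowTexture, though no texture sample appears.
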Material /QuestUI/QuestScreen/Materials/M_UI_TestBackground
graph {
 properties {bAutomaticallySetUsageInEditor=false, materialDomain=MD_UI, usedShadingModels=Default Lit}
 nodes [
  Composite() {subgraphName=Fix widget aspect ratio}
  n1 = ScalarParameter() {group=Texture, parameterName=GradientOffset_X}
  n2 = ScalarParameter() {group=Texture, parameterName=GradientOffset_Y}
  n3 = MaterialFunctionCall(
   UVs=Subtract(A=NamedRerouteUsage(), B=AppendVector(A=@n1, B=@n2))
   Texture Scale=AppendVector(
    A=ScalarParameter() {defaultValue=1, group=Texture, parameterName=GradientScale_X}
    B=ScalarParameter() {defaultValue=1, group=Texture, parameterName=GradientScale_Y}
   )
  ) {materialFunction=&ScaleUVsByCenter}
  n4 = ScalarParameter() {defaultValue=1, group=Texture, parameterName=RadialGradientRadius}
  n5 = ScalarParameter() {defaultValue=2.33, group=Texture, parameterName=TextureGradientDensity}
  LinearInterpolate(
   A=NamedRerouteUsage()
   B=LinearInterpolate(
    A=NamedRerouteUsage()
    Alpha=MaterialFunctionCall(UVs=@n3.UVs, Radius=@n4, Density=@n5) {materialFunction=&RadialGradientExponential}.RadialGradientExponential
   )
   Alpha=ScalarParameter() {parameterName=IsShowTexture}
  ) {constAlpha=1}
  n6 = MaterialFunctionCall() {materialFunction=&GetUserInterfaceUV}
  n7 = AppendVector(A=Constant() {r=1}, B=Divide(A=ComponentMask(None=@n6.Pixel Size) {g=true}, B=ComponentMask(None=@n6.Pixel Size) {r=true})) {desc=Widget Vertical Aspect Ratio}
  MaterialExpressionReroute_1 = Reroute(None=@n7)
  n8 = Constant() {r=1}
  n9 = Multiply(A=TextureCoordinate(), B=Divide(A=@MaterialExpressionReroute_1.Output, B=AppendVector(A=@n8, B=@n8) {desc=Texture Vertical Aspect Ratio}))
  n10 = Divide(B=ScalarParameter() {defaultValue=1.8769, group=Elipse, parameterName=Scale_X})
  n11 = Multiply(A=ComponentMask(None=@n9) {r=true}, B=@n10) {constB=0.5}
  n12 = Divide(A=@n10)
  n13 = ScalarParameter() {defaultValue=0.382, group=Elipse, parameterName=OffsetX}
  n14 = Divide(B=ScalarParameter() {defaultValue=0.776, group=Elipse, parameterName=Scale_Y})
  n15 = Multiply(A=ComponentMask(None=@n9) {g=true}, B=@n14) {constB=0.5}
  n16 = Divide(A=@n14)
  n17 = ScalarParameter() {defaultValue=-0.352, group=Elipse, parameterName=OffsetY}
  n18 = AppendVector(
   A=Subtract(A=Subtract(A=@n11, B=@n12), B=Multiply(A=@n10, B=@n13)) {constB=0}
   B=Subtract(A=Subtract(A=@n15, B=@n16), B=Multiply(A=@n14, B=@n17)) {constB=0}
  )
  n19 = MaterialFunctionCall(
   UVs=@n18
   CenterPosition=Constant()
   Radius=ScalarParameter() {defaultValue=0.3272, group=Elipse, parameterName=MRadius}
   Density=ScalarParameter() {defaultValue=1, group=Elipse, parameterName=MDensity}
  ) {materialFunction=&RadialGradientExponential}
  n20 = MaterialFunctionCall(
   UVs=@n18
   CenterPosition=Constant()
   Radius=ScalarParameter() {defaultValue=0.4775, group=Elipse, parameterName=Radius}
   Density=ScalarParameter() {defaultValue=3, group=Elipse, parameterName=Density}
  ) {materialFunction=&RadialGradientExponential}
  LinearInterpolate(A=@n19.RadialGradientExponential, B=@n20.RadialGradientExponential, Alpha=@n20.RadialGradientExponential)
  n21 = LinearInterpolate(
   A=NamedRerouteUsage()
   B=LinearInterpolate(
    A=VectorParameter() {group=Elipse, parameterName=Color2, defaultValue=<rgba 0.0423 0.0027 0.0212 1>}.RGB
    B=VectorParameter() {group=Elipse, parameterName=Color1, defaultValue=<rgba 0.855 0.0056 0.0242 1>}.RGB
    Alpha=@n19.RadialGradientExponential
   )
   Alpha=LinearInterpolate(A=@n20.RadialGradientExponential, Alpha=ScalarParameter() {group=Elipse, parameterName=AnimElipseTransition}) {constB=0}
  ) {constA=1}
  n22 = VectorParameter() {parameterName=FadeColor, defaultValue=<rgba 1 0 0.0118 0>}
  n23 = Multiply(A=@n22.A) {constB=0.6}
  n24 = Constant() {r=-0.6}
  n25 = Constant() {r=0.4}
  n26 = Constant() {r=-0.05}
  n27 = MaterialFunctionCall!(
   Colored Gradient=LinearInterpolate(
    A=LinearInterpolate(A=@n21, B=@n22.RGB, Alpha=@n23)
    B=@n21
    Alpha=MaterialFunctionCall(GradientStart=@n24, GradientEnd=@n25, RotationAngle=@n26) {materialFunction=&MF_UI_LinearGradient}.Result
   ) {constB=0}
   "Saturation (-1 to 1)"=Constant() {r=1}
   Saturation Mask
   Saturation Mask Threshold=Constant() {r=95}
  ) {
   desc="Apply the saturation to the gradient A to B only if Color C doesn't cover the saturation\r\nDo not apply it if Color C covers the saturation because in that case it will be applied next to the 3 colors."
   materialFunction=/FortniteCoreUI/UI/Quests/Materials/MF_UI_Gradient_TransitionSaturation
  }
  NamedRerouteDeclaration(None=@n27.Colorful RGBA Gradient) {name=Iteration2BG, nodeColor=<rgba 0 0 0 1>}
  NamedRerouteDeclaration(None=VectorParameter() {group=Background, parameterName=MainColor, defaultValue=<rgba 0 0.0118 0.0627 1>}.RGB) {name=Background, nodeColor=<rgba 0 0 0 1>}
  PinBase()
  PinBase!(Output=@n7) {reroutePins [Output {expression=&self:MaterialExpressionReroute_1}]}
 ]
}
