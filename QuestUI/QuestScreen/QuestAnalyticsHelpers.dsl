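; Helper functions that report Quest screen UI events to Flow Analytics: main view opened, sub view opened, element interacted and impressions.
; View models supply analytics IDs, and validity checks guard their use.
; Impressions commit only for a valid category and non-empty array; interactions attach a "Content" ID when available.
; Sub view events require a product and are named by product ID, or "product / category" when a category is valid.
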(fn FireQuestAnalytics_CommitImpressions
  (params (CategoryInterfaceVM :type Quest%20UI%20Category%20Interface%20View%20Model%20Object%20Reference) (Impressions :type Array%20of%20Flow%20Analytics%20Impression%20Section%20Content%20Structures))
  (Utilities|IsValid CategoryInterfaceVM
    (:Is%20Valid
      (if (not (Utilities|Array|IsEmpty Impressions))
        (FlowAnalytics|FireEventAddImpression(Non-Localized) (Class|QuestUIBaseContentViewModel|GetAnalyticsID CategoryInterfaceVM) Impressions)))))

(fn FireQuestAnalytics_ElementInteracted
  (params (PlayerController :type Player%20Controller%20Object%20Reference) (Action :type String%20%28by%20ref%29) (ContentVM :type Quest%20UI%20Base%20Content%20View%20Model%20Object%20Reference))
  (bind _asfort_player_controller (Utilities|Casting|CastToFortPlayerController PlayerController))
  (Utilities|IsValid ContentVM
    (:Is%20Valid
      (FlowAnalytics|FireEventElementInteracted(Non-Localized) _asfort_player_controller Action (FlowAnalytics|AddString(Non-Localized) (Utilities|Struct|MakeFlowAnalyticsAdditionalInfo) "Content" (Class|QuestUIBaseContentViewModel|GetAnalyticsID ContentVM))))
    (:Is%20Not%20Valid
      (FlowAnalytics|FireEventElementInteracted(Non-Localized) _asfort_player_controller Action))))

(fn FireQuestAnalytics_MainViewOpened
  (params (QuestScreen :type Quest%20UIScreen%20Object%20Reference))
  (FlowAnalytics|FireEventMainViewOpened QuestScreen))

(fn FireQuestAnalytics_SubViewOpened
  (params (PlayerController :type Player%20Controller%20Object%20Reference) (ProductVM :type Quest%20UI%20Product%20View%20Model%20Object%20Reference) (CategoryInterfaceVM :type Quest%20UI%20Category%20Interface%20View%20Model%20Object%20Reference))
  (bind _returnvalue (Class|QuestUIBaseContentViewModel|GetAnalyticsID ProductVM))
  (bind _asfort_player_controller (Utilities|Casting|CastToFortPlayerController PlayerController))
  (Utilities|IsValid ProductVM
    (:Is%20Valid
      (Utilities|IsValid CategoryInterfaceVM
        (:Is%20Valid
          (FlowAnalytics|FireEventSubViewOpened(Non-Localized) _asfort_player_controller (Utilities|String|ToString(Text) (Utilities|Text|FormatText "INVTEXT(\"{0} / {1}\")" _returnvalue (Class|QuestUIBaseContentViewModel|GetAnalyticsID CategoryInterfaceVM)))) ; Product & Category
        )
        (:Is%20Not%20Valid
          (FlowAnalytics|FireEventSubViewOpened(Non-Localized) _asfort_player_controller _returnvalue) ; Only Product
        )))))
