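# Quest UI settings modal, a CommonActivatableWidget on the UIKit dialog that offers a 1, 2 or 3 choice, apparently for tracked quest options.
# PreConstruct feeds a tracked-quest options viewmodel to the dialog, binds its button clicks and applies compact layout on mobile.
# Selecting a toggle stores SelectedAmmount and clears the other two.
# Dialog button index 1 deactivates the widget; no graph here applies the stored amount.
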WidgetBlueprint /QuestUI/QuestScreen/Modal/WBP_QuestUISettingsModal
graphs [
 EventGraph {nodes [
   n1 = Get Main!(self:WBP UIKit Button Generic Base Object Reference=Get ButtonWithToggle_1() -> (ButtonWithToggle_1:WBP UIKit Button With Toggle Object Reference)) -> (Main:Grid Panel Object Reference)
   SetGridSlotColumnAndRow!(execute, Widget:Widget Object Reference=@n1, Row:int=1, Row Span:int=0, Column:int=0, Column Span:int=0) -> (then)
   n2 = SlotAsGridSlot(Widget=@n1)
   n3 = Get!(self:Grid Slot Object Reference=@n2) -> (Padding_Left:float=~, Padding_Top:float=~, Padding_Right:float=~, Padding_Bottom:float=~)
   > Set!(
    execute
    Padding_Left:float=Get MainPaddingLeft() -> (MainPaddingLeft:float)
    Padding_Top:float=@n3.Padding_Top
    Padding_Right:float=0
    Padding_Bottom:float=@n3.Padding_Bottom
    self:Grid Slot Object Reference=@n2
   ) -> (then, Output_Get:Margin Structure)
   n4 = "On Selected Changed (ButtonWithToggle_1)"!() -> (OutputDelegate, then, Button:Common Button Base Object Reference, Selected:bool) {is=title K2Node_ComponentBoundEvent}
   > Branch(Condition=@n4.Selected)
   > Set SelectedAmmount(SelectedAmmount:int=1) -> (Output_Get:int=0)
   n5 = Get ButtonWithToggle_2() -> (ButtonWithToggle_2:WBP UIKit Button With Toggle Object Reference)
   n6 = Get ButtonWithToggle_3() -> (ButtonWithToggle_3:WBP UIKit Button With Toggle Object Reference)
   > ClearSelection(self:Common Button Base Object Reference=[@n5, @n6])
   n7 = "On Selected Changed (ButtonWithToggle_2)"!() -> (OutputDelegate, then, Button:Common Button Base Object Reference, Selected:bool) {is=title K2Node_ComponentBoundEvent}
   > Branch(Condition=@n7.Selected)
   > Set SelectedAmmount(SelectedAmmount=2) -> (Output_Get=0)
   n8 = Get ButtonWithToggle_1()
   > ClearSelection(self=[@n8, @n6])
   n9 = "On Selected Changed (ButtonWithToggle_3)"!() -> (OutputDelegate, then, Button:Common Button Base Object Reference, Selected:bool) {is=title K2Node_ComponentBoundEvent}
   > Branch(Condition=@n9.Selected)
   > Set SelectedAmmount(SelectedAmmount=3) -> (Output_Get=0)
   > ClearSelection(self=[@n8, @n5])
   n10 = CustomEvent!() -> (OutputDelegate, then, ButtonIndex:int) {is=title K2Node_CustomEvent}
   > n11 = Switch on Int!(execute:Exec, Selection:int=@n10.ButtonIndex) -> (Default:Exec, 0:Exec, 1:Exec, 2:Exec)
   DeactivateWidget(execute=[@n13.OutputPin, @n11.1])
   Event PreConstruct()
   > n12 = Construct Object from Class(Class=&/VM_Dialog_TrackedQuestOptionsModal.VM_Dialog_TrackedQuestOptionsModal_C) -> (ReturnValue:VM Dialog Tracked Quest Options Modal Object Reference)
   > SetDialogData!(
    execute
    self:WBP UIKit Dialog Internal Object Reference=Get Dialog_Internal() -> (Dialog_Internal:WBP UIKit Dialog Internal Object Reference)
    Data:UI Kit Dialog Viewmodel Object Reference=@n12
   ) -> (then)
   > Bind Event to Dialog Button Clicked Dispatcher(self:WBP UIKit Dialog Internal Object Reference=Get Dialog_Internal(), Delegate=@n10) {is=title K2Node_AddDelegate}
   > SetCompactUIStyle() {is=title K2Node_CallFunction}
   n13 = Reroute Node()
   Event Construct()
   Event Tick()
   Get SelectedAmmount() -> (SelectedAmmount:int=0)
  ]}
 SetCompactUIStyle {nodes [
   SetCompactUIStyle() {is=title K2Node_FunctionEntry}
   > Branch(Condition=IsMobileGame())
   n1 = Get ButtonWithToggle_1()
   > SetVerticalAlignment(
    self:Grid Slot Object Reference=SlotAsGridSlot(Widget=Get Block_Text!(self:WBP UIKit Button Generic Base Object Reference=@n1) -> (Block_Text:WBP UIKit Block Text Object Reference))
    InVerticalAlignment=VAlign_Center
   )
   > SetPadding!(
    execute
    self:Grid Slot Object Reference=SlotAsGridSlot(Widget=@n1)
    InPadding_Left:float=0
    InPadding_Top:float=0
    InPadding_Right:float=0
    InPadding_Bottom:float=24
   ) -> (then)
   n2 = Get ButtonWithToggle_2()
   > SetVerticalAlignment(self=SlotAsGridSlot(Widget=Get Block_Text(self=@n2)), InVerticalAlignment=VAlign_Center)
   > SetPadding(self=SlotAsGridSlot(Widget=@n2), InPadding_Left=0, InPadding_Top=0, InPadding_Right=0, InPadding_Bottom=24)
   n3 = Get ButtonWithToggle_3()
   > SetVerticalAlignment(self=SlotAsGridSlot(Widget=Get Block_Text(self=@n3)), InVerticalAlignment=VAlign_Center)
  ]}
]
variables [SelectedAmmount {type=integer}]
widgets {
 info {
  inheritedWidgetCount=0
  namedSlotCount=1
  parentClass=&CommonActivatableWidget
  rootWidgetClass=/FortUILibrary/Systems/Dialog/WBP_UIKit_Dialog_Internal.WBP_UIKit_Dialog_Internal_C
  widgetCount=6
 }
 widgets [
  Dialog_Internal {
   widgetClassPath=/FortUILibrary/Systems/Dialog/WBP_UIKit_Dialog_Internal
   bIsVariable=true
   widget=Dialog_Internal
   properties {buttons {
     each Button Horizontal Fill=true
     each Button Min Width=0
     each Button Size=Medium
     focuse Button=0
     space Between Buttons=30
     cTA Buttons Types [Regular,Quiet]
     group Placement {
      horizontal Alignment=HAlign_Fill
      vertical Alignment=VAlign_Bottom
      grid {column=1, column Span=0, layer=0, row=2, row Span=0, nudge=<0 0>}
      padding=<ltrb 80 64 80 48>
     }
    }}
  }
  GridPanel_0 {widgetClassPath=&GridPanel, namedSlotHost=Dialog_Internal, widget=GridPanel_0}
  GridPanel_145 {
   widgetClassPath=&GridPanel
   parent=GridPanel_0
   widget=GridPanel_145
   slot {class=GridSlot, properties {padding=<ltrb 82 100 0 0>}}
  }
  ButtonWithToggle_1 {
   widgetClassPath=/FortUILibrary/UIKit/Buttons/WBP_UIKit_ButtonWithToggle
   bIsVariable=true
   parent=GridPanel_145
   widget=ButtonWithToggle_1
   properties {text=1}
   slot {class=GridSlot}
  }
  ButtonWithToggle_2 {
   widgetClassPath=&WBP_UIKit_ButtonWithToggle
   bIsVariable=true
   parent=GridPanel_145
   widget=ButtonWithToggle_2
   properties {text=2}
   slot {class=GridSlot, properties {row=1}}
  }
  ButtonWithToggle_3 {
   widgetClassPath=&WBP_UIKit_ButtonWithToggle
   bIsVariable=true
   parent=GridPanel_145
   widget=ButtonWithToggle_3
   properties {text=3}
   slot {class=GridSlot, properties {row=2}}
  }
 ]
}
