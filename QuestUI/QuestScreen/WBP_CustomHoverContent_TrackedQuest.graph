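# CommonUI activatable widget serving as hover content for the Tracked Quests list in the quest screen, showing a title, tracked count and description.
# SetOrderedQuests stores the length of the supplied view model array, then calls Display, which also runs on activation.
# Display formats that count against the tracker view model's maximum, reached through the player view model, and switches the description depending on whether any quests are tracked.
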WidgetBlueprint /QuestUI/QuestScreen/WBP_CustomHoverContent_TrackedQuest
graphs [
 Display {nodes [
   Display() {is=title K2Node_FunctionEntry}
   n1 = Get TrackerVM!(self:Quest UI Player View Model Object Reference=Get QuestUIPlayerVM() -> (QuestUIPlayerVM:Quest UI Player View Model Object Reference)) -> (TrackerVM:Quest UI Tracker View Model Object Reference)
   > "SetText (Text)"(
    self=Get Text_TrackedCount() -> (Text_TrackedCount:Common Text Object Reference)
    InText=Format Text!(
     Format="LOCTEXT(\"{num}/{max}\")"
     num:int=Get NumberTrackedQuests() -> (NumberTrackedQuests:int=0)
     max:int=GetMaxTrackedQuestCount!(self:Quest UI Tracker View Model Object Reference=@n1) -> (ReturnValue:int=0)
    ) -> (Result)
   )
   n2 = Get NumberTrackedQuests() -> (NumberTrackedQuests=0)
   > "SetText (Text)"(
    self=Get Text_CoverDescription() -> (Text_CoverDescription:Common Text Object Reference)
    InText=Select(
     Option 0:Text="LOCTEXT(\"Track some quests to see them in your Tracked Quests list\")"
     Option 1:Text="LOCTEXT(\"Quests you are tracking throughout all your islands\")"
     Index="integer > integer"(A=@n2)
    ) -> (ReturnValue:Text)
   )
  ]}
 EventGraph {nodes [
   Event On Activated()
   > Display() {is=title K2Node_CallFunction}
   Event Construct()
  ]}
 SetOrderedQuests {nodes [
   n1 = SetOrderedQuests!() -> (then, OrderedQuests:Array of Quest UI Base Content View Model Object References) {is=title K2Node_FunctionEntry}
   > Set NumberTrackedQuests(NumberTrackedQuests:int=Length(TargetArray:Array of Quest UI Base Content View Model Object References=@n1)) -> (Output_Get:int=0)
   > Display()
  ]}
]
variables [
 MaximumTrackedQuests {type=integer}
 NumberTrackedQuests {type=integer}
]
widgets {
 info {inheritedWidgetCount=0, parentClass=&CommonActivatableWidget, rootWidgetClass=&GridPanel, widgetCount=6}
 widgets [
  GridPanel_0 {widgetClassPath=&GridPanel, widget=GridPanel_0, properties {columnFill [0,1]}}
  Img_TrackedQuestCover {
   widgetClassPath=&Image
   bIsVariable=true
   parent=GridPanel_0
   widget=Img_TrackedQuestCover
   properties {
    renderOpacity=0.2
    visibility=Collapsed
    brush {
     bIsDynamicallyLoaded=false
     drawAs=NoDrawType
     imageType=NoImage
     mirroring=NoMirror
     resourceName=None
     resourceObject=null
     tiling=NoTile
     imageSize=<512 512>
     margin=<ltrb 0 0 0 0>
     outlineSettings {
      bUseBrushTransparency=false
      roundingType=HalfHeightRadius
      width=0
      color {colorUseRule=UseColor_Specified, specifiedColor=<rgba 0 0 0 0>}
      cornerRadii=<0 0 0 0>
     }
     tintColor {colorUseRule=UseColor_Specified, specifiedColor=<rgba 1 1 1 1>}
     uVRegion {bIsValid=false, max=<0 0>, min=<0 0>}
    }
   }
   slot {class=GridSlot, properties {columnSpan=5, horizontalAlignment=HAlign_Center, verticalAlignment=VAlign_Center}}
  }
  Text_CoverTitle {
   widgetClassPath=&CommonTextBlock
   parent=GridPanel_0
   widget=Text_CoverTitle
   properties {
    renderOpacity=0.5
    text=Tracked Quests
    textTransformPolicy=ToUpper
    wrapTextAt=500
    font {
     bForceMonospaced=false
     bMaterialIsStencil=false
     fontMaterial=null
     fontObject=&HeadingNow
     letterSpacing=0
     monospacedWidth=1
     size=15.75
     skewAmount=0
     typefaceFontName=86 Bold Italic
     outlineSettings {
      bApplyOutlineToDropShadows=false
      bMiteredCorners=false
      bSeparateFillAlpha=false
      outlineMaterial=null
      outlineSize=0
      outlineColor=<rgba 0 0 0 1>
     }
    }
    shadowOffset=<0 0>
   }
   slot {class=GridSlot, properties {columnSpan=5, horizontalAlignment=HAlign_Left, row=1, verticalAlignment=VAlign_Bottom, padding=<ltrb 0 0 0 5>}}
  }
  Text_TrackedCount {
   widgetClassPath=&CommonTextBlock
   bIsVariable=true
   parent=GridPanel_0
   widget=Text_TrackedCount
   properties {
    lineHeightPercentage=0.9
    text=0/25
    textTransformPolicy=ToUpper
    wrapTextAt=500
    font {
     bForceMonospaced=true
     bMaterialIsStencil=false
     fontMaterial=null
     fontObject=&HeadingNow
     letterSpacing=0
     monospacedWidth=1
     size=30
     skewAmount=0
     typefaceFontName=76 Bold Italic
     outlineSettings {
      bApplyOutlineToDropShadows=false
      bMiteredCorners=false
      bSeparateFillAlpha=false
      outlineMaterial=null
      outlineSize=0
      outlineColor=<rgba 0 0 0 1>
     }
    }
    shadowOffset=<0 0>
   }
   slot {class=GridSlot, properties {row=2, verticalAlignment=VAlign_Bottom, padding=<ltrb 0 0 8 0>}}
  }
  Text_TrackedCount_1 {
   widgetClassPath=&CommonTextBlock
   parent=GridPanel_0
   widget=Text_TrackedCount_1
   properties {
    text=max
    wrapTextAt=500
    font {
     bForceMonospaced=false
     bMaterialIsStencil=false
     fontMaterial=null
     fontObject=&HeadingNow
     letterSpacing=0
     monospacedWidth=1
     size=21
     skewAmount=0
     typefaceFontName=75 Medium Italic
     outlineSettings {
      bApplyOutlineToDropShadows=false
      bMiteredCorners=false
      bSeparateFillAlpha=false
      outlineMaterial=null
      outlineSize=0
      outlineColor=<rgba 0 0 0 1>
     }
    }
    shadowOffset=<0 0>
   }
   slot {class=GridSlot, properties {column=1, row=2, verticalAlignment=VAlign_Bottom}}
  }
  Text_CoverDescription {
   widgetClassPath=&CommonTextBlock
   bIsVariable=true
   parent=GridPanel_0
   widget=Text_CoverDescription
   properties {
    text=This Quest cannot be progressed on the selected island.
    wrapTextAt=500
    font {
     bForceMonospaced=false
     bMaterialIsStencil=false
     fontMaterial=null
     fontObject=&HeadingNow
     letterSpacing=0
     monospacedWidth=1
     size=18
     skewAmount=0
     typefaceFontName=74 Regular
     outlineSettings {
      bApplyOutlineToDropShadows=false
      bMiteredCorners=false
      bSeparateFillAlpha=false
      outlineMaterial=null
      outlineSize=0
      outlineColor=<rgba 0 0 0 1>
     }
    }
    shadowOffset=<0 0>
   }
   slot {class=GridSlot, properties {columnSpan=5, horizontalAlignment=HAlign_Left, row=3, verticalAlignment=VAlign_Bottom, padding=<ltrb 0 13 0 0>}}
  }
 ]
}
