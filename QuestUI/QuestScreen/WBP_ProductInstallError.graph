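# Quest screen widget that shows a product's icon and name above a warning icon and a message asking the player to install or update it to view its quests.
# SetInstallState takes a product view model, checks validity, then fills both texts and the icon from it, loading the icon from a soft material.
# A Select on install state picks the message: update for NeedsUpdate, empty for Installed, install for NotInstalled, Invalid and Count.
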WidgetBlueprint /QuestUI/QuestScreen/WBP_ProductInstallError
graphs [
 EventGraph {links [], nodes []}
 SetInstallState {nodes [
   n1 = SetInstallState!() -> (then, ProductVM:Quest UI Product View Model Object Reference) {is=title K2Node_FunctionEntry}
   > Is Valid(InputObject=@n1)
   n2 = MakeLiteralText(Value="LOCTEXT(\"Please install this product to view these quests\")")
   > "SetText (Text)"(
    self=Get Text_InstallStatus() -> (Text_InstallStatus:Common Text Object Reference)
    InText=Select!(
     Installed:Text
     NotInstalled:Text=@n2
     NeedsUpdate:Text="LOCTEXT(\"Please update this product to view these quests\")"
     Invalid:Text=@n2
     Count:Text=@n2
     Index:EVerseURIInstallState Enum=Get ProductInstallState!(self:Quest UI Product View Model Object Reference=@n1) -> (ProductInstallState:EVerseURIInstallState Enum=Installed)
    ) -> (ReturnValue:Text)
   )
   > "SetText (Text)"(
    self=Get ProductName() -> (ProductName:Common Text Object Reference)
    InText=Get ProductName!(self:Quest UI Product View Model Object Reference=@n1) -> (ProductName:Text)
   )
   > SetBrushFromSoftMaterial!(
    execute
    self:Image Object Reference=Get ProductIcon() -> (ProductIcon:Fort Lazy Image Object Reference)
    SoftMaterial:Material Interface Soft Object Reference=Get ProductIcon!(self:Quest UI Product View Model Object Reference=@n1) -> (ProductIcon:Material Instance Soft Object Reference)
   ) -> (then)
  ]}
]
variables []
widgets {
 info {inheritedWidgetCount=0, parentClass=&UserWidget, rootWidgetClass=&Overlay, widgetCount=8}
 widgets [
  Overlay_19 {widgetClassPath=&Overlay, widget=Overlay_19}
  VerticalBox_0 {
   widgetClassPath=&VerticalBox
   parent=Overlay_19
   widget=VerticalBox_0
   slot {class=OverlaySlot, properties {horizontalAlignment=HAlign_Fill}}
  }
  HorizontalBox {widgetClassPath=&HorizontalBox, parent=VerticalBox_0, widget=HorizontalBox, slot {class=VerticalBoxSlot}}
  ProductIcon {
   widgetClassPath=&FortLazyImage
   bIsVariable=true
   parent=HorizontalBox
   widget=ProductIcon
   properties {brush {
     bIsDynamicallyLoaded=false
     drawAs=Image
     imageType=FullColor
     mirroring=NoMirror
     resourceName=None
     resourceObject=/FortUILibrary/Textures/Material/MI_UIKit_Icon_Base
     tiling=NoTile
     imageSize=<48 48>
     margin=<ltrb 0 0 0 0>
     outlineSettings {
      bUseBrushTransparency=false
      roundingType=HalfHeightRadius
      width=0
      color {colorUseRule=UseColor_Specified, specifiedColor=<rgba 0 0 0 0>}
      cornerRadii=<0 0 0 0>
     }
     tintColor {colorUseRule=UseColor_Specified, specifiedColor=<rgba 1 1 1 1>}
     uVRegion {bIsValid=false, max=<0 0>, min=<0 0>}
    }}
   slot {class=HorizontalBoxSlot, properties {horizontalAlignment=HAlign_Center, verticalAlignment=VAlign_Center}}
  }
  ProductName {
   widgetClassPath=&CommonTextBlock
   bIsVariable=true
   parent=HorizontalBox
   widget=ProductName
   properties {
    bIsScrollingEnabled=false
    text=Product Name
    font {
     bForceMonospaced=false
     bMaterialIsStencil=false
     fontMaterial=null
     fontObject=&HeadingNow
     letterSpacing=-40
     monospacedWidth=1
     size=32.25
     skewAmount=0
     typefaceFontName=86 Bold
     outlineSettings {
      bApplyOutlineToDropShadows=false
      bMiteredCorners=false
      bSeparateFillAlpha=false
      outlineMaterial=null
      outlineSize=0
      outlineColor=<rgba 0 0 0 1>
     }
    }
    shadowOffset=<0 0>
   }
   slot {class=HorizontalBoxSlot, properties {verticalAlignment=VAlign_Center, padding=<ltrb 10 0 0 0>}}
  }
  HorizontalBox_0 {
   widgetClassPath=&HorizontalBox
   parent=VerticalBox_0
   widget=HorizontalBox_0
   slot {class=VerticalBoxSlot, properties {padding=<ltrb 30 20 0 0>}}
  }
  FortLazyImage_236 {
   widgetClassPath=&FortLazyImage
   bIsVariable=true
   parent=HorizontalBox_0
   widget=FortLazyImage_236
   properties {brush {
     bIsDynamicallyLoaded=false
     drawAs=Image
     imageType=FullColor
     mirroring=NoMirror
     resourceName=None
     resourceObject=/Game/UI/Foundation/Shell/Textures/T-Icon-Warning-32
     tiling=NoTile
     imageSize=<32 32>
     margin=<ltrb 0 0 0 0>
     outlineSettings {
      bUseBrushTransparency=false
      roundingType=HalfHeightRadius
      width=0
      color {colorUseRule=UseColor_Specified, specifiedColor=<rgba 0 0 0 0>}
      cornerRadii=<0 0 0 0>
     }
     tintColor {colorUseRule=UseColor_Specified, specifiedColor=<rgba 1 1 1 1>}
     uVRegion {bIsValid=false, max=<0 0>, min=<0 0>}
    }}
   slot {class=HorizontalBoxSlot}
  }
  Text_InstallStatus {
   widgetClassPath=&CommonTextBlock
   bIsVariable=true
   parent=HorizontalBox_0
   widget=Text_InstallStatus
   properties {
    bIsScrollingEnabled=false
    style=/FortUILibrary/Styles/Typography/Foundation/TextStyle-Base-S.TextStyle-Base-S_C
    text=Install status
    font {
     bForceMonospaced=false
     bMaterialIsStencil=false
     fontMaterial=null
     fontObject=&NotoSans
     letterSpacing=0
     monospacedWidth=1
     size=18
     skewAmount=0
     typefaceFontName=NotoSans-Regular
     outlineSettings {
      bApplyOutlineToDropShadows=false
      bMiteredCorners=false
      bSeparateFillAlpha=false
      outlineMaterial=null
      outlineSize=0
      outlineColor=<rgba 0 0 0 1>
     }
    }
    shadowOffset=<0 0>
   }
   slot {class=HorizontalBoxSlot, properties {verticalAlignment=VAlign_Center, padding=<ltrb 10 0 0 0>}}
  }
 ]
}
