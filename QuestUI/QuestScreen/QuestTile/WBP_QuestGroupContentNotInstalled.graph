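# Quest screen list entry that reports the download status of a quest group's uninstalled content and offers a button to the lobby.
# OnListItemObjectSet casts the item to FortLibraryStorageItemViewModel and sets it on the widget.
# Setter functions, with no callers in the graphs, fill title, icon and percentage, and map status to description, button label and progress visibility.
# Clicking finds the parent AthenaUIStateWidget_Frontend and calls NavigateToLobbyPedestals.
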WidgetBlueprint /QuestUI/QuestScreen/QuestTile/WBP_QuestGroupContentNotInstalled
graphs [
 EventGraph {nodes [
   "On Clicked (WBP_UIKit_Button_Regular)"!() -> (OutputDelegate, then, Button:Common Button Base Object Reference) {is=title K2Node_ComponentBoundEvent}
   > n1 = FindParentWidgetOfType!(
    execute
    StartingWidget:Widget Object Reference=Self-Reference()
    Type:Widget Class Reference=/Script/FortniteUI.AthenaUIStateWidget_Frontend
   ) -> (then, ReturnValue:Athena UIState Widget Frontend Object Reference)
   > NavigateToLobbyPedestals(self:Athena UIState Widget Frontend Object Reference=@n1)
   n2 = Event OnListItemObjectSet()
   > n3 = Cast To FortLibraryStorageItemViewModel!(execute, Object=@n2.ListItemObject) -> (then, CastFailed, AsFort Library Storage Item View Model:Fort Library Storage Item View Model Object Reference) {is=title K2Node_DynamicCast}
   > Set FortLibraryStorageItemViewModel(FortLibraryStorageItemViewModel:Fort Library Storage Item View Model Object Reference=@n3) -> (Output_Get:Fort Library Storage Item View Model Object Reference)
  ]}
 SetDownloadStatus {nodes [
   n1 = SetDownloadStatus!() -> (then, Status:ELibraryStorageItemDownloadStatus Enum=NotDownloaded) {is=title K2Node_FunctionEntry}
   > "SetText (Text)"(
    self=Get Description() -> (Description:Common Text Object Reference)
    InText=Select!(
     NotDownloaded:Text="LOCTEXT(\"To access the product, install additional content\")"
     Downloaded:Text
     Downloading:Text="LOCTEXT(\"The product is being uploaded, please wait\")"
     UpdateRequired:Text="LOCTEXT(\"Need update\")"
     Updating:Text="LOCTEXT(\"The product is being uploaded, please wait\")"
     Queued:Text="LOCTEXT(\"The product is being uploaded, please wait\")"
     Paused:Text="LOCTEXT(\"Updating on a pause\")"
     PendingRemoval:Text
     Removing:Text="LOCTEXT(\"To access the product, install additional content\")"
     Index:ELibraryStorageItemDownloadStatus Enum=@n1
    ) -> (ReturnValue:Text)
   )
   > SetText!(
    execute
    self:Fort CTAButton Object Reference=Get WBP_UIKit_Button_Regular() -> (WBP_UIKit_Button_Regular:WBP UIKit Button Regular Object Reference)
    ButtonText:"Text (by ref)"=Select(
     NotDownloaded="LOCTEXT(\"Download from lobby\")"
     Downloading="LOCTEXT(\"Go to lobby\")"
     UpdateRequired="LOCTEXT(\"Download from lobby\")"
     Updating="LOCTEXT(\"Go to lobby\")"
     Queued="LOCTEXT(\"Go to lobby\")"
     Paused="LOCTEXT(\"Download from lobby\")"
     Removing="LOCTEXT(\"Download from lobby\")"
     Index=@n1
    )
   ) -> (then)
   > SetVisibility(
    self=[
     Get Percentage() -> (Percentage:Common Text Object Reference)
     Get WBP_UIKit_Throbber() -> (WBP_UIKit_Throbber:WBP UIKit Throbber Object Reference)
    ]
    InVisibility=Select(
     NotDownloaded:ESlateVisibility Enum=Collapsed
     Downloaded:ESlateVisibility Enum=Collapsed
     Downloading:ESlateVisibility Enum=SelfHitTestInvisible
     UpdateRequired:ESlateVisibility Enum=Collapsed
     Updating:ESlateVisibility Enum=SelfHitTestInvisible
     Queued:ESlateVisibility Enum=SelfHitTestInvisible
     Paused:ESlateVisibility Enum=SelfHitTestInvisible
     PendingRemoval:ESlateVisibility Enum=Visible
     Removing:ESlateVisibility Enum=Collapsed
     Index=@n1
    ) -> (ReturnValue:ESlateVisibility Enum)
   )
  ]}
 SetProductIcon {nodes [
   n1 = SetProductIcon!() -> (then, Icon:Texture 2D Object Reference) {is=title K2Node_FunctionEntry}
   > SetBrushFromTexture!(
    execute
    self:Image Object Reference=Get Icon() -> (Icon:Image Object Reference)
    Texture:Texture 2D Object Reference=@n1
    bMatchSize:bool=false
   ) -> (then)
  ]}
 SetTitle {nodes [
   n1 = SetTitle!() -> (then, Title:Text) {is=title K2Node_FunctionEntry}
   > "SetText (Text)"(self=Get Header() -> (Header:Common Text Object Reference), InText=@n1)
  ]}
 UpdateProgress {nodes [
   n1 = UpdateProgress!() -> (then, Progress:int) {is=title K2Node_FunctionEntry}
   > "SetText (Text)"(self=Get Percentage(), InText=Format Text!(Format="LOCTEXT(\"{x}%\")", x:int=@n1) -> (Result))
  ]}
]
variables []
widgets {
 info {inheritedWidgetCount=0, parentClass=&CommonUserWidget, rootWidgetClass=&GridPanel, widgetCount=7}
 widgets [
  Root {widgetClassPath=&GridPanel, widget=Root, properties {columnFill [1,0], rowFill [1]}}
  Icon {
   widgetClassPath=&Image
   bIsVariable=true
   parent=Root
   widget=Icon
   properties {brush {
     bIsDynamicallyLoaded=false
     drawAs=Image
     imageType=FullColor
     mirroring=NoMirror
     resourceName=None
     resourceObject=/FortUILibrary/Textures/Material/MI_UIKit_Icon_Base
     tiling=NoTile
     imageSize=<32 32>
     margin=<ltrb 0 0 0 0>
     outlineSettings {
      bUseBrushTransparency=false
      roundingType=HalfHeightRadius
      width=0
      color {colorUseRule=UseColor_Specified, specifiedColor=<rgba 0 0 0 0>}
      cornerRadii=<0 0 0 0>
     }
     tintColor {colorUseRule=UseColor_Specified, specifiedColor=<rgba 1 1 1 1>}
     uVRegion {bIsValid=false, max=<0 0>, min=<0 0>}
    }}
   slot {class=GridSlot, properties {horizontalAlignment=HAlign_Left, verticalAlignment=VAlign_Center, padding=<ltrb 0 0 16 0>}}
  }
  Header {
   widgetClassPath=&CommonTextBlock
   bIsVariable=true
   parent=Root
   widget=Header
   properties {
    style=&TextStyle-Base_C
    text=Text Block
    font {
     bForceMonospaced=false
     bMaterialIsStencil=false
     fontMaterial=null
     fontObject=&NotoSans
     letterSpacing=0
     monospacedWidth=1
     size=24
     skewAmount=0
     typefaceFontName=NotoSans-Regular
     outlineSettings {
      bApplyOutlineToDropShadows=false
      bMiteredCorners=false
      bSeparateFillAlpha=false
      outlineMaterial=null
      outlineSize=0
      outlineColor=<rgba 0 0 0 1>
     }
    }
    shadowOffset=<0 0>
   }
   slot {class=GridSlot, properties {columnSpan=1, padding=<ltrb 40 0 0 6>}}
  }
  Description {
   widgetClassPath=&CommonTextBlock
   bIsVariable=true
   parent=Root
   widget=Description
   properties {
    style=&TextStyle-Base_C
    text=Text Block
    font {
     bForceMonospaced=false
     bMaterialIsStencil=false
     fontMaterial=null
     fontObject=&NotoSans
     letterSpacing=0
     monospacedWidth=1
     size=24
     skewAmount=0
     typefaceFontName=NotoSans-Regular
     outlineSettings {
      bApplyOutlineToDropShadows=false
      bMiteredCorners=false
      bSeparateFillAlpha=false
      outlineMaterial=null
      outlineSize=0
      outlineColor=<rgba 0 0 0 1>
     }
    }
    shadowOffset=<0 0>
   }
   slot {class=GridSlot, properties {columnSpan=1, row=1, padding=<ltrb 0 0 0 10>}}
  }
  WBP_UIKit_Button_Regular {
   widgetClassPath=&WBP_UIKit_Button_Regular
   bIsVariable=true
   parent=Root
   widget=WBP_UIKit_Button_Regular
   properties {text=Download from lobby}
   slot {class=GridSlot, properties {horizontalAlignment=HAlign_Left, row=3, padding=<ltrb 0 8 0 0>}}
  }
  Percentage {
   widgetClassPath=&CommonTextBlock
   bIsVariable=true
   parent=Root
   widget=Percentage
   properties {
    style=&TextStyle-Base_C
    text=99%
    font {
     bForceMonospaced=false
     bMaterialIsStencil=false
     fontMaterial=null
     fontObject=&NotoSans
     letterSpacing=0
     monospacedWidth=1
     size=24
     skewAmount=0
     typefaceFontName=NotoSans-Regular
     outlineSettings {
      bApplyOutlineToDropShadows=false
      bMiteredCorners=false
      bSeparateFillAlpha=false
      outlineMaterial=null
      outlineSize=0
      outlineColor=<rgba 0 0 0 1>
     }
    }
    shadowOffset=<0 0>
   }
   slot {class=GridSlot, properties {row=2, verticalAlignment=VAlign_Center, padding=<ltrb 52 0 0 0>}}
  }
  WBP_UIKit_Throbber {
   widgetClassPath=/FortUILibrary/UIKit/Throbber/WBP_UIKit_Throbber
   bIsVariable=true
   parent=Root
   widget=WBP_UIKit_Throbber
   slot {class=GridSlot, properties {horizontalAlignment=HAlign_Left, row=2}}
  }
 ]
}
