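# Quest screen row widget for the tracked quest category, a child of WBP_QuestCategoryRow_Entry that declares no widgets or variables of its own.
# On PreConstruct it clears existing entries and creates a single WBP_QuestCategoryTile in the DynamicEntryBox_Content box.
# It applies the tracked-quest style to that tile and binds its button events through the row's BindButtonEvents.
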WidgetBlueprint /QuestUI/QuestScreen/QuestTile/WBP_TrackedQuestCategory_Entry
graphs [EventGraph {nodes [
   Event PreConstruct()
   > ClearAllEntries(self:WBP Quest Category Row Entry Object Reference)
   > n1 = Create Entry of Class!(
    execute
    self:Dynamic Entry Box Object Reference=Get DynamicEntryBox_Content() -> (DynamicEntryBox_Content:Dynamic Entry Box Object Reference)
    EntryClass:User Widget Class Reference=&/WBP_QuestCategoryTile.WBP_QuestCategoryTile_C
   ) -> (then, ReturnValue:WBP Quest Category Tile Object Reference)
   > SetTrackedQuestStyle!(execute, self:WBP Quest Category Tile Object Reference=@n1, IsTrackedQuestTile:bool=true) -> (then)
   > BindButtonEvents!(execute, self:WBP Quest Category Row Entry Object Reference, ButtonInstance:Common Button Base Object Reference=@n1, Bind:bool=true) -> (then)
  ]}]
variables []
widgets {
 info {
  inheritedWidgetCount=0
  parentClass=&/WBP_QuestCategoryRow_Entry.WBP_QuestCategoryRow_Entry_C
  rootWidgetClass=null
  widgetCount=0
 }
 widgets []
}
