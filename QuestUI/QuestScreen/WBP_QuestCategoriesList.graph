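# Quest screen category list, a Fort polymorphic list view that picks entry widgets by view model type: category rows, headers, spacers, not-installed content, collections and tracker.
# Remembers the focused column and applies it to the newly selected category row.
# Replaces mouse-wheel scrolling with whole-item steps that skip non-selectable headers and spacers.
# Snaps to top or bottom a tick after the second or second-to-last item scrolls into view.
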WidgetBlueprint /QuestUI/QuestScreen/WBP_QuestCategoriesList
graphs [
 EventGraph {nodes [
   n1 = "On Item Selection Changed (ListView)"!() -> (OutputDelegate, then, Item:Object Reference, bIsSelected:bool) {is=title K2Node_ComponentBoundEvent}
   > Branch(Condition=@n1.bIsSelected)
   > n2 = Cast To WBP_QuestCategoryRow_Entry!(
    execute
    Object=GetWidgetFromItem!(self:Fort Polymorphic View Object Reference, Item:Object Reference=@n1.Item) -> (ReturnValue:User Widget Object Reference)
   ) -> (then, CastFailed, AsWBP Quest Category Row Entry:WBP Quest Category Row Entry Object Reference) {is=title K2Node_DynamicCast}
   > SetActiveColumn!(execute, self:WBP Quest Category Row Entry Object Reference=@n2, Column:int=Get ActiveColumn() -> (ActiveColumn:int=0)) -> (then)
   n3 = "On Item Scrolled Into View (ListView)"!() -> (OutputDelegate, then, Item:Object Reference, Widget:User Widget Object Reference) {is=title K2Node_ComponentBoundEvent}
   n4 = Get ListView() -> (ListView:Common List View Object Reference)
   n5 = GetIndexForItem!(self:List View Object Reference=@n4, Item:Object Reference=@n3.Item) -> (ReturnValue:int=0)
   > n6 = Branch(Condition="Equal (Integer)"(A=@n5, B=1))
   > DelayUntilNextTick!(execute) -> (then)
   > ScrollToTop(self:List View Base Object Reference=@n4)
   n7 = GetNumItems!(self:List View Object Reference=@n4) -> (ReturnValue:int=0)
   Branch(execute=@n6.else, Condition="Equal (Integer)"(A=@n5, B=int - int(A=@n7, B=2)))
   > DelayUntilNextTick()
   > ScrollToBottom(self:List View Base Object Reference=@n4)
   n8 = "On Entry Generated (ListView)"!() -> (OutputDelegate, then, Widget:User Widget Object Reference) {is=title K2Node_ComponentBoundEvent}
   > n9 = Cast To WBP_QuestCategoryRow_Entry(Object=@n8.Widget)
   n10 = HandleColumnFocused!() -> (OutputDelegate, then, Column:int) {is=title K2Node_CustomEvent}
   Bind Event to On Column Focused(execute=@n9.then, self:WBP Quest Category Row Entry Object Reference=@n9, Delegate=@n10) {is=title K2Node_AddDelegate}
   Set ActiveColumn(execute=@n10.then, ActiveColumn:int=@n10.Column) -> (Output_Get:int=0)
  ]}
 GetDesiredEntryClassForItem {nodes [
   n1 = GetDesiredEntryClassForItem!() -> (then, Item:Object Reference) {is=title K2Node_FunctionEntry}
   > n2 = Cast To QuestUICategoryRowVM!(execute, Object=@n1) -> (then, CastFailed, AsQuest UI Category Row View Model:Quest UI Category Row View Model Object Reference) {is=title K2Node_DynamicCast}
   > Return Node!(execute, ReturnValue:User Widget Class Reference=/QuestUI/QuestScreen/QuestTile/WBP_QuestCategoryRow_Entry.WBP_QuestCategoryRow_Entry_C)
   n3 = Cast To DebugQuestUIHeaderVM!(execute=@n2.CastFailed, Object=@n1) -> (then, CastFailed, AsDebug Quest UIHeader VM:Debug Quest UIHeader VM Object Reference) {is=title K2Node_DynamicCast}
   > Return Node(ReturnValue=/QuestUI/QuestScreen/QuestTile/WBP_QuestGroupSeparatorTitle_Entry.WBP_QuestGroupSeparatorTitle_Entry_C)
   n4 = Cast To QuestUISpacerVM!(execute=@n3.CastFailed, Object=@n1) -> (then, CastFailed, AsQuest UISpacer VM:Quest UISpacer VM Object Reference) {is=title K2Node_DynamicCast}
   > Return Node(ReturnValue=/QuestUI/QuestScreen/QuestTile/WBP_QuestSpacer_Entry.WBP_QuestSpacer_Entry_C)
   n5 = Cast To FortLibraryStorageItemViewModel!(execute=@n4.CastFailed, Object=@n1) -> (then, CastFailed, AsFort Library Storage Item View Model:Fort Library Storage Item View Model Object Reference) {is=title K2Node_DynamicCast}
   > Return Node(ReturnValue=/QuestUI/QuestScreen/QuestTile/WBP_QuestGroupContentNotInstalled.WBP_QuestGroupContentNotInstalled_C)
   n6 = Cast To QuestUICollectionsVM!(execute=@n5.CastFailed, Object=@n1) -> (then, CastFailed, AsQuest UI Collections View Model:Quest UI Collections View Model Object Reference) {is=title K2Node_DynamicCast}
   > Return Node(ReturnValue=/QuestUI/QuestScreen/QuestTile/WBP_CollectionsButton_Entry.WBP_CollectionsButton_Entry_C)
   Cast To QuestUITrackerVM!(execute=@n6.CastFailed, Object=@n1) -> (then, CastFailed, AsQuest UI Tracker View Model:Quest UI Tracker View Model Object Reference) {is=title K2Node_DynamicCast}
   > Return Node(ReturnValue=/QuestUI/QuestScreen/QuestTile/WBP_TrackedQuestCategory_Entry.WBP_TrackedQuestCategory_Entry_C)
  ]}
 GetItemIsSelectable {nodes [
   n1 = GetItemIsSelectable!() -> (then, Item:Object Reference) {is=title K2Node_FunctionEntry}
   > n2 = Cast To DebugQuestUIHeaderVM(Object=@n1)
   Return Node(execute=[@n3.then, @n2.then], ReturnValue:bool=false)
   n3 = Cast To QuestUISpacerVM(execute=@n2.CastFailed, Object=@n1)
   Return Node(execute=@n3.CastFailed, ReturnValue=true)
  ]}
 GetItemSelectableOrNavigable {nodes [
   n1 = GetItemSelectableOrNavigable!() -> (then, Item:Object Reference) {is=title K2Node_FunctionEntry}
   > n2 = GetItemIsSelectable!(execute, self, Item:Object Reference=@n1) -> (then, ReturnValue:bool) {is=title K2Node_CallFunction}
   > Return Node(ReturnValue=@n2)
  ]}
 OnMouseWheel {nodes [
   n1 = OnMouseWheel!() -> (then, MyGeometry:Geometry Structure, MouseEvent:"Pointer Event Structure (by ref)") {is=title K2Node_FunctionEntry}
   Set NewOffset(execute=[@n1.then, @n9.else], NewOffset:int=0) -> (Output_Get:int=0)
   > n2 = Branch(Condition="float > float"(A=Get Wheel Delta!(Input:"Pointer Event Structure (by ref)"=@n1.MouseEvent) -> (ReturnValue:float) {is=title K2Node_CallFunction}))
   n3 = GetScrollOffset!(self:List View Base Object Reference=Get ListView()) -> (ReturnValue:float)
   > n4 = Branch(Condition="float > float"(A=@n3))
   n5 = int - int(A=Floor(A=@n3) {is=title K2Node_CallFunction})
   > Set NewOffset(NewOffset=@n5) -> (Output_Get=0)
   n6 = Get ListView()
   > n7 = SetScrollOffset!(execute, self:List View Base Object Reference=@n6, InScrollOffset:float="To Float (Integer)"(InInt=@n5)) -> (then)
   n8 = GetItemIsSelectable(
    execute=[@n14.then, @n7.then]
    Item=GetItemAt!(self:List View Object Reference=@n6, Index:int=Get NewOffset() -> (NewOffset:int=0)) -> (ReturnValue:Object Reference)
   )
   > n9 = Branch(Condition=@n8)
   > Return Node(ReturnValue:Event Reply Structure=Handled())
   Return Node(execute=@n4.else, ReturnValue=Handled())
   n10 = Get ListView()
   n11 = Get ListView()
   n12 = Branch(execute=@n2.else, Condition="float < float"(A=GetScrollOffset(self=@n10), B:int=GetNumItems(self=@n11) -> (ReturnValue=0)))
   n13 = int + int(A=Ceil(A=GetScrollOffset(self=@n11)) {is=title K2Node_CallFunction})
   > Set NewOffset(NewOffset=@n13) -> (Output_Get=0)
   > n14 = SetScrollOffset(self=Get ListView(), InScrollOffset="To Float (Integer)"(InInt=@n13))
   Return Node(execute=@n12.else, ReturnValue=Handled())
  ]}
]
variables [ActiveColumn {type=integer}]
widgets {
 info {
  inheritedWidgetCount=1
  parentClass=/Script/FortniteUI.FortPolymorphicListView
  rootWidgetClass=&CommonListView
  widgetCount=1
 }
 widgets [ListView {
   widgetClassPath=&CommonListView
   bInherited=true
   bIsVariable=true
   widget=ListView
   properties {
    bEnableRightClickScrolling=false
    bReturnFocusToSelection=true
    clipping=Inherit
    consumeMouseWheel=Never
    entryWidgetClass=&WBP_QuestCategoryRow_Entry_C
    verticalEntrySpacing=24
    wheelScrollMultiplier=0
    scrollBarStyle {
     thickness=-2
     draggedThumbImage {
      bIsDynamicallyLoaded=false
      drawAs=NoDrawType
      imageType=FullColor
      mirroring=NoMirror
      resourceName=../../../Engine/Content/Slate/Common/Scrollbar_Thumb.png
      resourceObject=null
      tiling=NoTile
      imageSize=<0 0>
      margin=<ltrb 0.25 0.25 0.25 0.25>
      outlineSettings {
       bUseBrushTransparency=false
       roundingType=HalfHeightRadius
       width=0
       color {colorUseRule=UseColor_Specified, specifiedColor=<rgba 0 0 0 0>}
       cornerRadii=<0 0 0 0>
      }
      tintColor {colorUseRule=UseColor_Specified, specifiedColor=<rgba 1 1 1 1>}
      uVRegion {bIsValid=false, max=<0 0>, min=<0 0>}
     }
     horizontalBackgroundImage {
      bIsDynamicallyLoaded=false
      drawAs=NoDrawType
      imageType=NoImage
      mirroring=NoMirror
      resourceName=None
      resourceObject=null
      tiling=NoTile
      imageSize=<0 0>
      margin=<ltrb 0 0 0 0>
      outlineSettings {
       bUseBrushTransparency=false
       roundingType=HalfHeightRadius
       width=0
       color {colorUseRule=UseColor_Specified, specifiedColor=<rgba 0 0 0 0>}
       cornerRadii=<0 0 0 0>
      }
      tintColor {colorUseRule=UseColor_Specified, specifiedColor=<rgba 1 1 1 1>}
      uVRegion {bIsValid=false, max=<0 0>, min=<0 0>}
     }
     horizontalBottomSlotImage {
      bIsDynamicallyLoaded=false
      drawAs=NoDrawType
      imageType=FullColor
      mirroring=NoMirror
      resourceName=../../../Engine/Content/Slate/Common/Scrollbar_Background_Horizontal.png
      resourceObject=null
      tiling=NoTile
      imageSize=<8 8>
      margin=<ltrb 0 0 0 0>
      outlineSettings {
       bUseBrushTransparency=false
       roundingType=HalfHeightRadius
       width=0
       color {colorUseRule=UseColor_Specified, specifiedColor=<rgba 0 0 0 0>}
       cornerRadii=<0 0 0 0>
      }
      tintColor {colorUseRule=UseColor_Specified, specifiedColor=<rgba 1 1 1 1>}
      uVRegion {bIsValid=false, max=<0 0>, min=<0 0>}
     }
     horizontalTopSlotImage {
      bIsDynamicallyLoaded=false
      drawAs=NoDrawType
      imageType=FullColor
      mirroring=NoMirror
      resourceName=../../../Engine/Content/Slate/Common/Scrollbar_Background_Horizontal.png
      resourceObject=null
      tiling=NoTile
      imageSize=<8 8>
      margin=<ltrb 0 0 0 0>
      outlineSettings {
       bUseBrushTransparency=false
       roundingType=HalfHeightRadius
       width=0
       color {colorUseRule=UseColor_Specified, specifiedColor=<rgba 0 0 0 0>}
       cornerRadii=<0 0 0 0>
      }
      tintColor {colorUseRule=UseColor_Specified, specifiedColor=<rgba 1 1 1 1>}
      uVRegion {bIsValid=false, max=<0 0>, min=<0 0>}
     }
     hoveredThumbImage {
      bIsDynamicallyLoaded=false
      drawAs=NoDrawType
      imageType=FullColor
      mirroring=NoMirror
      resourceName=../../../Engine/Content/Slate/Common/Scrollbar_Thumb.png
      resourceObject=null
      tiling=NoTile
      imageSize=<0 0>
      margin=<ltrb 0.25 0.25 0.25 0.25>
      outlineSettings {
       bUseBrushTransparency=false
       roundingType=HalfHeightRadius
       width=0
       color {colorUseRule=UseColor_Specified, specifiedColor=<rgba 0 0 0 0>}
       cornerRadii=<0 0 0 0>
      }
      tintColor {colorUseRule=UseColor_Specified, specifiedColor=<rgba 1 1 1 1>}
      uVRegion {bIsValid=false, max=<0 0>, min=<0 0>}
     }
     normalThumbImage {
      bIsDynamicallyLoaded=false
      drawAs=NoDrawType
      imageType=FullColor
      mirroring=NoMirror
      resourceName=../../../Engine/Content/Slate/Common/Scrollbar_Thumb.png
      resourceObject=null
      tiling=NoTile
      imageSize=<0 0>
      margin=<ltrb 0.25 0.25 0.25 0.25>
      outlineSettings {
       bUseBrushTransparency=false
       roundingType=HalfHeightRadius
       width=0
       color {colorUseRule=UseColor_Specified, specifiedColor=<rgba 0 0 0 0>}
       cornerRadii=<0 0 0 0>
      }
      tintColor {colorUseRule=UseColor_Specified, specifiedColor=<rgba 1 1 1 1>}
      uVRegion {bIsValid=false, max=<0 0>, min=<0 0>}
     }
     verticalBackgroundImage {
      bIsDynamicallyLoaded=false
      drawAs=NoDrawType
      imageType=NoImage
      mirroring=NoMirror
      resourceName=None
      resourceObject=null
      tiling=NoTile
      imageSize=<0 0>
      margin=<ltrb 0 0 0 0>
      outlineSettings {
       bUseBrushTransparency=false
       roundingType=HalfHeightRadius
       width=0
       color {colorUseRule=UseColor_Specified, specifiedColor=<rgba 0 0 0 0>}
       cornerRadii=<0 0 0 0>
      }
      tintColor {colorUseRule=UseColor_Specified, specifiedColor=<rgba 1 1 1 1>}
      uVRegion {bIsValid=false, max=<0 0>, min=<0 0>}
     }
     verticalBottomSlotImage {
      bIsDynamicallyLoaded=false
      drawAs=NoDrawType
      imageType=FullColor
      mirroring=NoMirror
      resourceName=../../../Engine/Content/Slate/Common/Scrollbar_Background_Vertical.png
      resourceObject=null
      tiling=NoTile
      imageSize=<8 8>
      margin=<ltrb 0 0 0 0>
      outlineSettings {
       bUseBrushTransparency=false
       roundingType=HalfHeightRadius
       width=0
       color {colorUseRule=UseColor_Specified, specifiedColor=<rgba 0 0 0 0>}
       cornerRadii=<0 0 0 0>
      }
      tintColor {colorUseRule=UseColor_Specified, specifiedColor=<rgba 1 1 1 1>}
      uVRegion {bIsValid=false, max=<0 0>, min=<0 0>}
     }
     verticalTopSlotImage {
      bIsDynamicallyLoaded=false
      drawAs=NoDrawType
      imageType=FullColor
      mirroring=NoMirror
      resourceName=../../../Engine/Content/Slate/Common/Scrollbar_Background_Vertical.png
      resourceObject=null
      tiling=NoTile
      imageSize=<8 8>
      margin=<ltrb 0 0 0 0>
      outlineSettings {
       bUseBrushTransparency=false
       roundingType=HalfHeightRadius
       width=0
       color {colorUseRule=UseColor_Specified, specifiedColor=<rgba 0 0 0 0>}
       cornerRadii=<0 0 0 0>
      }
      tintColor {colorUseRule=UseColor_Specified, specifiedColor=<rgba 1 1 1 1>}
      uVRegion {bIsValid=false, max=<0 0>, min=<0 0>}
     }
    }
   }
  }]
}
