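# Quest screen side navigation widget that lists quest products as collapsible tabs, with a Clear Badges button.
# InitSideNavigation builds tabs and a divider from product view models; selection sets the minimap component's active product context and broadcasts the product.
# Hover, gamepad focus and external requests feed HandleCollapseRequest, which syncs tabs, Clear Badges and a gamepad back button that refocuses the selected tab.
# Mobile uses compact styling and never collapses.
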WidgetBlueprint /QuestUI/QuestScreen/WBP_QuestScreen_SideNavigation
graphs [
 EventGraph {nodes [
   n1 = "On Side Navigation Collapsed State Changed (SideNavigation)"!() -> (OutputDelegate, then, IsCollapsed:bool) {is=title K2Node_ComponentBoundEvent}
   > HandleCollapseRequest!(execute, self, ShouldCollapse:bool=@n1.IsCollapsed, IsImmediate:bool=false) -> (then)
   > Call On Collapsed State Changed!(execute:Exec, self:WBP Quest Screen Side Navigation Object Reference, IsCollapsed:bool=@n1.IsCollapsed) -> (then:Exec) {is=title K2Node_CallDelegate}
   "On Clicked (Button_ClearBadges)"!() -> (OutputDelegate, then, Button:Common Button Base Object Reference) {is=title K2Node_ComponentBoundEvent}
   > Call On Clear Badges Clicked!(execute:Exec, self:WBP Quest Screen Side Navigation Object Reference) -> (then:Exec) {is=title K2Node_CallDelegate}
   n2 = "On Unhovered (Button_Hitbox)"() {is=title K2Node_ComponentBoundEvent}
   HandleCollapseRequest(execute=[@n2.then, @n4.then], ShouldCollapse=true, IsImmediate=false)
   n3 = "On Hovered (Button_ClearBadges)"!() -> (OutputDelegate, then, Button:Common Button Base Object Reference) {is=title K2Node_ComponentBoundEvent}
   HandleCollapseRequest(execute=[@n3.then, @n8.then], ShouldCollapse=false, IsImmediate=false)
   n4 = "On Unhovered (Button_ClearBadges)"!() -> (OutputDelegate, then, Button:Common Button Base Object Reference) {is=title K2Node_ComponentBoundEvent}
   "On Clicked (Button_BackToSideNavigation)"!() -> (OutputDelegate, then, Button:Common Button Base Object Reference) {is=title K2Node_ComponentBoundEvent}
   n5 = Get SideNavigation() -> (SideNavigation:WBP UIKit Side Navigation Object Reference)
   > SetFocus(self=GetTabButtonBaseByID!(
     self:Common Tab List Widget Base Object Reference=@n5
     TabNameID:Name=GetSelectedTabId!(self:Common Tab List Widget Base Object Reference=@n5) -> (ReturnValue:Name=None)
    ) -> (ReturnValue:Common Button Base Object Reference))
   "On Tab Clicked (SideNavigation)"!() -> (OutputDelegate, then, TabId:Name=None) {is=title K2Node_ComponentBoundEvent}
   > Call On Product Clicked!(execute:Exec, self:WBP Quest Screen Side Navigation Object Reference) -> (then:Exec) {is=title K2Node_CallDelegate}
   n6 = "On Side Navigation Selection Changed (SideNavigation)"!() -> (OutputDelegate, then, TabID:Name=None, InnerTabID:Name=None, TabViewModel:MVVM Base Viewmodel Object Reference) {is=title K2Node_ComponentBoundEvent}
   > n7 = Cast To QuestUIProductVM!(execute, Object=@n6.TabViewModel) -> (then, CastFailed, AsQuest UI Product View Model:Quest UI Product View Model Object Reference) {is=title K2Node_DynamicCast}
   > SetActiveProductContext!(
    execute
    self:Fort Controller Component Minimap Object Reference=GetComponentByClass!(
     self:Actor Object Reference=GetOwningPlayer()
     ComponentClass:Actor Component Class Reference=/Script/FortniteGame.FortControllerComponent_Minimap
    ) -> (ReturnValue:Fort Controller Component Minimap Object Reference)
    ProductTag:"Gameplay Tag Structure (by ref)"=Get ProductGroupTag!(self:Quest UI Product View Model Object Reference=@n7) -> (ProductGroupTag:Gameplay Tag Structure)
   ) -> (then)
   > Call On Product Selected!(execute:Exec, self:WBP Quest Screen Side Navigation Object Reference, ProductVM:Quest UI Product View Model Object Reference=@n7) -> (then:Exec) {is=title K2Node_CallDelegate}
   n8 = "On Hovered (Button_Hitbox)"() {is=title K2Node_ComponentBoundEvent}
   Event OnAddedToFocusPath!() -> (OutputDelegate, then, InFocusEvent:Focus Event Structure) {is=title K2Node_Event}
   n9 = IsInputType!(InputType:ECommonInputType Enum=Gamepad, IsNot:bool=false) -> (Result:bool)
   > Branch(Condition=@n9)
   > Do Once(Reset=@n10.then)
   > DelayUntilNextTick!(execute) -> (then)
   > SetFocus(self=GetSelectedButton!() -> (ButtonInstance:Common Button Base Object Reference))
   > HandleCollapseRequest(ShouldCollapse=false, IsImmediate=false)
   Event OnRemovedFromFocusPath!() -> (OutputDelegate, then, InFocusEvent:Focus Event Structure) {is=title K2Node_Event}
   > n10 = Branch(Condition=@n9)
   n11 = Event SetIsCollapsed!() -> (OutputDelegate, then, IsCollapsed:bool, SkipAnimation:bool) {is=title K2Node_Event}
   > HandleCollapseRequest(ShouldCollapse=@n11.IsCollapsed, IsImmediate=@n11.SkipAnimation)
   Event OnInitialized()
   > SetCompactUIStyle() {is=title K2Node_CallFunction}
   Event Construct()
   n12 = Get FortGlobalUIContext() -> (ReturnValue:Fort Global UIContext Object Reference) {is=title K2Node_GetSubsystem}
   > SetVisibility(
    self=Get Button_BackToSideNavigation() -> (Button_BackToSideNavigation:WBP UIKit Button Generic Object Reference)
    InVisibility=Select(Option 0=SelfHitTestInvisible, Option 1=Collapsed, Index=IsInZone!(self:Fort Global UIContext Object Reference=@n12) -> (ReturnValue:bool))
   )
  ]}
 GetRemainingAnimTime {nodes [
   GetRemainingAnimTime() {is=title K2Node_FunctionEntry}
   > Return Node!(execute, FunctionIsImplemented:bool=false, RemainingTime:double=0, TotalTime:double=0)
  ]}
 GetRequiresVisibility {nodes [
   GetRequiresVisibility() {is=title K2Node_FunctionEntry}
   n1 = Get OrderedProductVMs!(self:Quest UI Player View Model Object Reference=Get QuestUIPlayerVM() -> (QuestUIPlayerVM:Quest UI Player View Model Object Reference)) -> (OrderedProductVMs:Array of Quest UI Product View Model Object References)
   > Return Node!(execute, RequiresVisibility:bool="integer > integer"(A=Length(TargetArray:Array of Quest UI Product View Model Object References=@n1), B=1))
  ]}
 GetSelectedButton {nodes [
   GetSelectedButton() {is=title K2Node_FunctionEntry}
   n1 = Get SideNavigation()
   > Return Node!(
    execute
    ButtonInstance:Common Button Base Object Reference=GetTabButtonBaseByID(self=@n1, TabNameID=GetSelectedTabId(self=@n1) -> (ReturnValue=None))
   )
  ]}
 GetSelectedProductVM {nodes [
   GetSelectedProductVM() {is=title K2Node_FunctionEntry}
   n1 = Get SideNavigation()
   > n2 = Cast To QuestUIProductVM(Object=GetTabViewModel!(self:UIKit Tab Group Object Reference=@n1, TabID:Name=GetSelectedTabId(self=@n1) -> (ReturnValue=None)) -> (ReturnValue:MVVM Base Viewmodel Object Reference))
   > Return Node!(execute, ProductVM:Quest UI Product View Model Object Reference=@n2)
   Return Node(execute=@n2.CastFailed)
  ]}
 HandleCollapseRequest {nodes [
   n1 = HandleCollapseRequest!() -> (then, ShouldCollapse:bool=~, IsImmediate:bool=~) {is=title K2Node_FunctionEntry}
   > n2 = Sequence()
   > Set LocalShouldCollapse(LocalShouldCollapse:bool=Select(Option 0:bool=@n1, Option 1:bool=false, Index=IsMobileGame()) -> (ReturnValue:bool)) -> (Output_Get:bool)
   n3 = Get IsCollapsed() -> (IsCollapsed:bool)
   n4 = Get LocalShouldCollapse() -> (LocalShouldCollapse:bool)
   Branch(
    execute=@n2.then_1
    Condition=OR Boolean(A="Not Equal (Boolean)"(A=@n4, B=@n3) -> (ReturnValue=~), B=Get IsImmediate() -> (IsImmediate:bool))
   )
   > n5 = Set IsCollapsed(IsCollapsed:bool=Get LocalShouldCollapse()) -> (Output_Get:bool)
   > n6 = Branch(Condition=@n5)
   n7 = Get SideNavigation()
   n8 = Get IsImmediate()
   > SetIsCollapsed!(execute, self:WBP UIKit Side Navigation Base Object Reference=@n7, IsCollapsed:bool=true, SkipAnimation:bool=@n8) -> (then)
   n9 = Get Button_ClearBadges() -> (Button_ClearBadges:WBP UIKit Button Collapsible Object Reference)
   > SetIsCollapsed(self:WBP UIKit Button Collapsible Object Reference=@n9, IsCollapsed=true, SkipAnimation=@n8)
   n10 = Get Button_BackToSideNavigation()
   > SetVisibility(self=@n10, InVisibility=SelfHitTestInvisible)
   SetIsCollapsed(execute=@n6.else, self:WBP UIKit Side Navigation Base Object Reference=@n7, IsCollapsed=false, SkipAnimation=@n8)
   > SetIsCollapsed(self:WBP UIKit Button Collapsible Object Reference=@n9, IsCollapsed=false, SkipAnimation=@n8)
   > SetVisibility(self=@n10, InVisibility=Collapsed)
  ]}
 HasNewContent {nodes [
   n1 = HasNewContent!() -> (then, NewContent:bool=~) {is=title K2Node_FunctionEntry}
   n2 = Get Button_ClearBadges()
   > SetIsLocked!(execute, self:Common Button Base Object Reference=@n2, bInIsLocked:bool=NOT Boolean(A=@n1)) -> (then)
   > SetRenderOpacity(self=@n2, InOpacity=SelectFloat(A=1, B=0.4, bPickA=@n1))
  ]}
 InitSideNavigation {nodes [
   n1 = InitSideNavigation!() -> (then, ProductVMs:Array of Quest UI Product View Model Object References) {is=title K2Node_FunctionEntry}
   > RemoveTabs(self:WBP UIKit Tab Group Base Object Reference=Get SideNavigation())
   > n2 = For Each Loop(Array:Array of Quest UI Product View Model Object References=@n1) -> (Array Element:Quest UI Product View Model Object Reference)
   n3 = Sequence(execute=@n2.Completed)
   n4 = GetDividerIndex!(self:Quest UI Player View Model Object Reference=Get QuestUIPlayerVM()) -> (ReturnValue:int=0)
   > n5 = Branch(Condition="integer >= integer"(A=@n4))
   > n6 = AddDividerByIndex!(execute, self:WBP UIKit Tab Group Base Object Reference=Get SideNavigation(), TabIndex:int=@n4, AddAfter:bool=false) -> (then)
   SelectTabByIndex!(execute=[@n6.then, @n5.else], self:UIKit Tab Group Object Reference=Get SideNavigation(), Index:int=0, bSuppressClickFeedback:bool=true) -> (then, ReturnValue:bool)
   Call On Product Selected(
    execute=@n3.then_1
    ProductVM="Get (a copy)"(Array:Array of Quest UI Product View Model Object References=@n1) -> (Output:Quest UI Product View Model Object Reference)
   )
   > Call On Requires Visibility Changed!(
    execute:Exec
    self:WBP Quest Screen Side Navigation Object Reference
    RequiresVisibility:bool=GetRequiresVisibility!() -> (RequiresVisibility:bool) {is=title K2Node_CallFunction}
   ) -> (then:Exec) {is=title K2Node_CallDelegate}
   AddSideNavigationTab!(
    execute=@n2.LoopBody
    self:WBP UIKit Side Navigation Base Object Reference=Get SideNavigation()
    TabViewModel:MVVM Base Viewmodel Object Reference=@n2
    TabID:Name=String To Name(InString="To String (Integer)"(InInt=@n2.Array Index))
    Index:int=-1
   ) -> (then)
  ]}
 OnClearBadgesClicked {nodes [OnClearBadgesClicked() {is=title K2Node_FunctionEntry}]}
 OnCollapsedStateChanged {nodes [OnCollapsedStateChanged!() -> (then, IsCollapsed:bool=~) {is=title K2Node_FunctionEntry}]}
 OnProductClicked {nodes [OnProductClicked() {is=title K2Node_FunctionEntry}]}
 OnProductSelected {nodes [OnProductSelected!() -> (then, ProductVM:Quest UI Product View Model Object Reference) {is=title K2Node_FunctionEntry}]}
 OnRequiresVisibilityChanged {nodes [OnRequiresVisibilityChanged!() -> (then, RequiresVisibility:bool=~) {is=title K2Node_FunctionEntry}]}
 SetClearBadgesVisibility {nodes [
   n1 = SetClearBadgesVisibility!() -> (then, Visibility:ESlateVisibility Enum=Visible) {is=title K2Node_FunctionEntry}
   > SetVisibility(self=Get Button_ClearBadges(), InVisibility=@n1)
  ]}
 SetCompactUIStyle {nodes [
   SetCompactUIStyle() {is=title K2Node_FunctionEntry}
   > Branch(Condition=IsMobileGame())
   n1 = Get SideNavigation()
   > Set!(execute, MinButtonSize_X:int=0, MinButtonSize_Y:int=80, self:WBP UIKit Tab Group Base Object Reference=@n1) -> (then, Output_Get:Int Point Structure)
   > Set UseMinButtonSize!(execute, UseMinButtonSize:bool=true, self:WBP UIKit Tab Group Base Object Reference=@n1) -> (then, Output_Get:bool)
   > Set EnableCollapsedState!(execute, EnableCollapsedState:bool=false, self:WBP UIKit Side Navigation Base Object Reference=@n1) -> (then, Output_Get:bool)
   > Set IconSize!(execute, IconSize:Vector 2D Structure=(X=48,Y=48), self:WBP UIKit Tab Group Base Object Reference=@n1) -> (then, Output_Get:Vector 2D Structure)
   > Set BackplateCornerRadius!(execute, BackplateCornerRadius:E_UI_BackplateCornerRadius Enum=NewEnumerator0, self:WBP UIKit Tab Group Base Object Reference=@n1) -> (then, Output_Get:E_UI_BackplateCornerRadius Enum=NewEnumerator0)
   > Set ScrollBarVisibility!(execute, ScrollBarVisibility:ESlateVisibility Enum=Visible, self:WBP UIKit Tab Group Base Object Reference=@n1) -> (then, Output_Get:ESlateVisibility Enum=Visible)
   > Set!(
    execute
    ScrollbarPadding_Left:float=8
    ScrollbarPadding_Top:float=16
    ScrollbarPadding_Right:float=16
    ScrollbarPadding_Bottom:float=16
    self:Scroll Box Object Reference=Get ScrollBox_Tabs!(self:WBP UIKit Tab Group Base Object Reference=@n1) -> (ScrollBox_Tabs:Scroll Box Object Reference)
   ) -> (then, Output_Get:Margin Structure)
   > Set IconSize(IconSize=(X=48,Y=48), self:WBP UIKit Button Generic Base Object Reference=Get Button_ClearBadges())
  ]}
]
variables [
 IsCollapsed {type=boolean}
 OnClearBadgesClicked {}
 OnCollapsedStateChanged {}
 OnProductClicked {}
 OnProductSelected {}
 OnRequiresVisibilityChanged {}
]
widgets {
 info {inheritedWidgetCount=0, parentClass=&UserWidget, rootWidgetClass=&Overlay, widgetCount=7}
 widgets [
  Overlay_0 {widgetClassPath=&Overlay, widget=Overlay_0}
  Button_Hitbox {
   widgetClassPath=&Button
   parent=Overlay_0
   widget=Button_Hitbox
   properties {renderOpacity=0}
   slot {class=OverlaySlot, properties {horizontalAlignment=HAlign_Fill, verticalAlignment=VAlign_Fill}}
  }
  VerticalBox_25 {widgetClassPath=&VerticalBox, parent=Overlay_0, widget=VerticalBox_25, slot {class=OverlaySlot}}
  SideNavigation {
   widgetClassPath=/FortUILibrary/UIKit/Tabs/WBP_UIKit_SideNavigation
   bIsVariable=true
   parent=VerticalBox_25
   widget=SideNavigation
   properties {
    buttonClass=&/WBP_Quests_SideNavigationButton.WBP_Quests_SideNavigationButton_C
    dividerClass=&/WBP_QuestsTabDivider.WBP_QuestsTabDivider_C
    dividerSpacing=4
    enableCollapsedState=true
    nextPromptHorizontalAlignment=HAlign_Right
    prevPromptHorizontalAlignment=HAlign_Left
    tabSpacing=8
    usePromptBackplate=false
   }
   slot {class=VerticalBoxSlot, properties {size {sizeRule=Fill, value=1}}}
  }
  CommonHardwareVisibilityBorder_0 {
   widgetClassPath=&CommonHardwareVisibilityBorder
   parent=VerticalBox_25
   widget=CommonHardwareVisibilityBorder_0
   properties {visibilityQuery {
     autoDescription=" ANY( Input.Gamepad )"
     tokenStreamVersion=0
     userDescription=""
     queryTokenStream [0,1,1,1,0]
     tagDictionary [{tagName=Input.Gamepad}]
    }}
   slot {class=VerticalBoxSlot}
  }
  Button_BackToSideNavigation {
   widgetClassPath=&WBP_UIKit_Button_Generic
   bIsVariable=true
   parent=CommonHardwareVisibilityBorder_0
   widget=Button_BackToSideNavigation
   properties {
    bHideInputActionWithKeyboard=true
    bIsFocusable=false
    minHeight=40
    promptHorizontalAlignment=HAlign_Center
    useBackgroundBlock=false
    useIconBlock=false
    useTextBlock=false
    visibility=Collapsed
    buttonPadding=<ltrb 0 0 0 0>
    promptNudge=<0 0>
    triggeringInputAction {dataTable=&InputActionDataTable, rowName=GenericBack}
   }
   slot {class=BorderSlot, properties {padding=<ltrb 0 0 0 0>}}
  }
  Button_ClearBadges {
   widgetClassPath=/FortUILibrary/UIKit/Buttons/WBP_UIKit_Button_Collapsible
   bIsVariable=true
   parent=VerticalBox_25
   widget=Button_ClearBadges
   properties {
    buttonType=Quiet
    iconMaterial=/FortUILibrary/Textures/MaterialInstances/MI_UIKit_Icon_ClearBadges
    maxTextWidth=214
    mobileFontSizeMultiplier=1.3
    mobileMinHeight=80
    overrideDefaultMobileFontSizeMultiplier=false
    text=Clear Badges
    buttonPadding=<ltrb 16 12 12 12>
    iconColorAndOpacityOverride=<rgba 0.0048 0.004 0.0116 1>
    mobileButtonPadding=<ltrb 16 0 16 0>
    triggeringInputAction {dataTable=&InputActionDataTable, rowName=GenericLeftStickClick}
   }
   slot {class=VerticalBoxSlot, properties {verticalAlignment=VAlign_Top, padding=<ltrb 0 24 0 0>}}
  }
 ]
}
