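# Quest screen widget that lists quest categories per product, shown in the frontend and on the in-game map screen.
# Side navigation sets the active product view model, which rebuilds the category list or shows an install error.
# Tile hover updates rewards or an async-loaded custom widget; clicks switch to a quests modal created asynchronously on activation.
# It restores saved selections, clears them by timer, listens for cinematic events and manages gamepad focus.
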WidgetBlueprint /QuestUI/QuestScreen/WBP_QuestScreen
animations [HideRewards {
  binding_count=1
  length_seconds=0.25
  object=&self:HideRewards
  sequence {bindings [Switcher_HoveredDetail {
     id=#1
     possessed_class=CommonVisibilitySwitcher
     tracks [{
       type=MovieSceneFloatTrack
       display_name=Render Opacity
       sections [{type=MovieSceneFloatSection, end_frame=5, start_frame=0, channels [None {type=MovieSceneScriptingFloatChannel, keys [[0,1],[5,0]]}]}]
      }]
    }]}
 }]
graphs [
 BP_GetDesiredFocusTarget {nodes [
   Get Desired Focus Target()
   n1 = Select(
    Option 0:Widget Object Reference=Get ListView!(self:Fort Polymorphic List View Object Reference=Get QuestCategoriesList() -> (QuestCategoriesList:WBP Quest Categories List Object Reference)) -> (ListView:Common List View Object Reference)
    Option 1:Widget Object Reference=Get SideNavigation() -> (SideNavigation:WBP Quest Screen Side Navigation Object Reference)
    Index:int=GetActiveWidgetIndex!(self:Widget Switcher Object Reference=Get ContentSwitcher() -> (ContentSwitcher:Common Activatable Widget Switcher Object Reference)) -> (ReturnValue:int=0)
   ) -> (ReturnValue:Widget Object Reference)
   > Return Node!(execute, ReturnValue:Widget Object Reference=@n1)
   n2 = Get QuestModalRef() -> (QuestModalRef:WBP Quests Modal Iteration 2 Object Reference)
   n3 = Get WidgetSwitcher() -> (WidgetSwitcher:Common Activatable Widget Switcher Object Reference)
   Select(
    Option 0=@n1
    Option 1=GetDesiredFocusTarget!(self:Common Activatable Widget Object Reference=@n2) -> (ReturnValue:Widget Object Reference)
    Index:bool="Equal (Object)"(A=GetActiveWidget!(self:Widget Switcher Object Reference=@n3) -> (ReturnValue:Widget Object Reference), B=@n2)
   )
  ]}
 BP_GetDesiredInputConfig {nodes [
   Get Desired Input Config() {is=title K2Node_FunctionEntry}
   > Return Node!(
    execute
    ReturnValue_bIgnoreMoveInput:bool=False
    ReturnValue_bIgnoreLookInput:bool=False
    ReturnValue_InputMode:ECommonInputMode Enum=All
    ReturnValue_MouseCaptureMode:EMouseCaptureMode Enum=NoCapture
    ReturnValue_MouseLockMode:EMouseLockMode Enum=DoNotLock
    ReturnValue_bHideCursorDuringViewportCapture:bool=True
   )
  ]}
 BindSidebarEvents {nodes [
   n1 = BindSidebarEvents!() -> (then, Bind:bool=~) {is=title K2Node_FunctionEntry}
   > n2 = Branch(Condition=@n1)
   n3 = Create Event()
   n4 = Get SideNavigation()
   > Bind Event to On Collapsed State Changed(self:WBP Quest Screen Side Navigation Object Reference=@n4, Delegate=@n3) {is=title K2Node_AddDelegate}
   n5 = Create Event()
   > Bind Event to On Clear Badges Clicked(self:WBP Quest Screen Side Navigation Object Reference=@n4, Delegate=@n5) {is=title K2Node_AddDelegate}
   n6 = Create Event()
   > Bind Event to On Product Selected(self:WBP Quest Screen Side Navigation Object Reference=@n4, Delegate=@n6) {is=title K2Node_AddDelegate}
   n7 = Create Event()
   > Bind Event to On Product Clicked(self:WBP Quest Screen Side Navigation Object Reference=@n4, Delegate=@n7) {is=title K2Node_AddDelegate}
   n8 = Create Event()
   > Bind Event to On Requires Visibility Changed(self:WBP Quest Screen Side Navigation Object Reference=@n4, Delegate=@n8) {is=title K2Node_AddDelegate}
   Unbind Event from On Collapsed State Changed(execute=@n2.else, self:WBP Quest Screen Side Navigation Object Reference=@n4, Delegate=@n3) {is=title K2Node_RemoveDelegate}
   > Unbind Event from On Clear Badges Clicked(self:WBP Quest Screen Side Navigation Object Reference=@n4, Delegate=@n5) {is=title K2Node_RemoveDelegate}
   > Unbind Event from On Product Selected(self:WBP Quest Screen Side Navigation Object Reference=@n4, Delegate=@n6) {is=title K2Node_RemoveDelegate}
   > Unbind Event from On Product Clicked(self:WBP Quest Screen Side Navigation Object Reference=@n4, Delegate=@n7) {is=title K2Node_RemoveDelegate}
   > Unbind Event from On Requires Visibility Changed(self:WBP Quest Screen Side Navigation Object Reference=@n4, Delegate=@n8) {is=title K2Node_RemoveDelegate}
  ]}
 CancelClearSelectionTimer {nodes [
   CancelClearSelectionTimer() {is=title K2Node_FunctionEntry}
   n1 = Get ClearSelectionTimerHandle() -> (ClearSelectionTimerHandle:Timer Handle Structure)
   > Branch(Condition=Is Valid Timer Handle!(Handle:Timer Handle Structure=@n1) -> (ReturnValue:bool))
   > Clear and Invalidate Timer by Handle(Handle=@n1)
  ]}
 ClearSelectionData {nodes [
   ClearSelectionData() {is=title K2Node_FunctionEntry}
   > Invalidate Timer Handle!(execute, Handle:"Timer Handle Structure (by ref)"=Get ClearSelectionTimerHandle()) -> (then, ReturnValue:Timer Handle Structure)
   n1 = Get QuestUIPlayerVM() -> (QuestUIPlayerVM:Quest UI Player View Model Object Reference)
   > Is Valid(InputObject=@n1)
   > ClearSelectionData(self:Quest UI Player View Model Object Reference=@n1) {is=title K2Node_CallFunction}
  ]}
 EventGraph {nodes [
   Fire Event Element Interacted()
   > n1 = Create Widget!(
    execute:Exec
    Class:User Widget Class Reference=/QuestUI/QuestScreen/Modal/WBP_QuestUISettingsModal.WBP_QuestUISettingsModal_C
    OwningPlayer:Player Controller Object Reference=GetOwningPlayer()
   ) -> (then:Exec, ReturnValue:WBP Quest UISettings Modal Object Reference)
   > TryPushWidgetAsModalPanel!(execute, PushWidget:User Widget Object Reference=@n1) -> (then)
   > ActivateWidget(self=@n1)
   SetDisableTransitionAnimation!(execute, self:Common Animated Switcher Object Reference=Get WidgetSwitcher(), bDisableAnimation:bool=true) -> (then)
   > n2 = For Each Loop with Break!(
    Exec:Exec
    Array:Array of Quest UI Category Interface View Model Object References=Get Category Tabs VM() -> (Category Tabs VM:Array of Quest UI Category Interface View Model Object References)
    Break:Exec=@n4.then
   ) -> (LoopBody:Exec, Array Element:Quest UI Category Interface View Model Object Reference, Array Index:int, Completed:Exec)
   > n3 = Branch(Condition="Equal (Integer)"(A=@n2.Array Index))
   n4 = Branch(execute=@n3.else, Condition=Get bContentEnabled!(self:Quest UI Base Content View Model Object Reference=@n2) -> (bContentEnabled:bool))
   PushQuestModal!(
    execute=@n2.Completed
    self
    CategoryInterfaceVM:Quest UI Category Interface View Model Object Reference=@n2
    ContentVM:Quest UI Base Content View Model Object Reference=Get ActiveQuestProductVM() -> (ActiveQuestProductVM:Quest UI Product View Model Object Reference)
    IsRestoringSelection:bool=false
   ) -> (then)
   "On Clicked (Button_ClearBadges_Gamepad)"!() -> (OutputDelegate, then, Button:Common Button Base Object Reference) {is=title K2Node_ComponentBoundEvent}
   > HandleSidebarClearBadgesClicked()
   "On Clicked (DecoyButton_DebugQuestUI)"!() -> (OutputDelegate, then, Button:Common Button Base Object Reference) {is=title K2Node_ComponentBoundEvent}
   > n5 = Branch(Condition="Equal Exactly (String)"(A=GetBuildConfiguration!() -> (ReturnValue:String), B=Get ShippingConfig() -> (ShippingConfig:String)))
   QueueModalWidgetSoftClassAsync!(execute:Exec=@n5.else, SoftWidgetClass:Common Activatable Widget Soft Class Reference=/QuestUI/DebugTool/QuestDebugTool.QuestDebugTool_C) -> (then:Exec, OnModalWidgetQueued:Exec, ActivatableWidget:Common Activatable Widget Object Reference) {is=title K2Node_AsyncAction}
   n6 = "On Active Widget Index Changed BP (WidgetSwitcher)"!() -> (OutputDelegate, then, ActiveWidget:Widget Object Reference, ActiveIndex:int=0) {is=title K2Node_ComponentBoundEvent}
   > Branch(Condition="Equal (Object)"(A=@n6.ActiveWidget, B=Get WBP_UIKit_FrontendLayout() -> (WBP_UIKit_FrontendLayout:WBP UIKit Frontend Layout Object Reference)))
   > RefreshFocus()
   n7 = OnModalDeactivated() {is=title K2Node_CustomEvent}
   > QuestModalDeactivated()
   > SetActiveWidget(self:Widget Switcher Object Reference=Get WidgetSwitcher(), Widget=Get WBP_UIKit_FrontendLayout())
   n8 = CustomEvent!() -> (OutputDelegate, then, TabID:Name, InnerTabID:Name, TabViewModel:MVVM Base Viewmodel Object Reference) {is=title K2Node_CustomEvent}
   > Set SelectedCategory(SelectedCategory:MVVM Base Viewmodel Object Reference=@n8.TabViewModel) -> (Output_Get:MVVM Base Viewmodel Object Reference)
   n9 = OnEntryHovered!() -> (OutputDelegate, then, Button:Common Button Base Object Reference) {is=title K2Node_CustomEvent}
   > n10 = Cast To WBP_QuestCategoryTile!(execute, Object=@n9.Button) -> (then, CastFailed, AsWBP Quest Category Tile:WBP Quest Category Tile Object Reference) {is=title K2Node_DynamicCast}
   > n11 = Branch(Condition=Get IsTrackedQuestTile!(self:WBP Quest Category Tile Object Reference=@n10) -> (IsTrackedQuestTile:bool))
   > SetActiveWidgetIndex(
    self:Common Visibility Switcher Object Reference=Get Switcher_HoveredDetail() -> (Switcher_HoveredDetail:Common Visibility Switcher Object Reference)
    Index=1
   )
   n12 = Get QuestUICategoryVM!(self:WBP Quest Category Tile Object Reference=@n10) -> (QuestUICategoryVM:Quest UI Category View Model Object Reference)
   n13 = Is Valid(exec=@n11.else, InputObject=@n12)
   UpdateRewards!(
    execute=[@n13.Is Not Valid, @n10.CastFailed]
    self:WBP Side Rewards Panel Object Reference=Get WBP_SideRewardsPanel() -> (WBP_SideRewardsPanel:WBP Side Rewards Panel Object Reference)
    RewardsCustomWidgets:Array of Quest UI Reward VM Object References
    QuestContentVM:Quest UI Base Content View Model Object Reference
   ) -> (then)
   > SetActiveWidgetIndex(self=Get Switcher_HoveredDetail(), Index=0)
   n14 = Get CustomMapWidget!(self:Quest UI Category View Model Object Reference=@n12) -> (CustomMapWidget:User Widget Soft Class Reference)
   UpdateCustomWidget!(execute=@n13.Is Valid, self, Custom Map Widget:User Widget Soft Class Reference=@n14) -> (then)
   > n15 = Branch(Condition=IsValidSoftClassReference!(SoftClassReference:"Object Soft Class Reference (by ref)"=@n14) -> (ReturnValue:bool))
   n16 = Set CategoryRewardVMs(
    execute=@n15.else
    CategoryRewardVMs:Array of Quest UI Reward VM Object References=Get CategoryRewardVMs!(self:Quest UI Category View Model Object Reference=@n12) -> (CategoryRewardVMs:Array of Quest UI Reward VM Object References)
   ) -> (Output_Get:Array of Quest UI Reward VM Object References)
   n17 = Get WBP_SideRewardsPanel()
   > UpdateRewards(self=@n17, RewardsCustomWidgets=@n16, QuestContentVM=@n12)
   > SetSelectedPage!(
    execute
    self:WBP Quest Pages Object Reference=Get Pages!(self:WBP Side Rewards Panel Object Reference=@n17) -> (Pages:WBP Quest Pages Object Reference)
    SelectedOption:int=0
   ) -> (then)
   > SetActiveWidgetIndex(self=Get Switcher_HoveredDetail(), Index=0)
   n18 = OnEntryClicked!() -> (OutputDelegate, then, Button:Common Button Base Object Reference) {is=title K2Node_CustomEvent}
   > n19 = Cast To WBP_QuestCategoryTile(Object=@n18.Button)
   > PushQuestModal(CategoryInterfaceVM=Get QuestUICategoryVM(self=@n19), IsRestoringSelection=false)
   n20 = UpdateCustomWidget!() -> (OutputDelegate, then, Custom Map Widget:User Widget Soft Class Reference) {is=title K2Node_CustomEvent}
   > n21 = Branch(Condition=IsValidSoftClassReference(SoftClassReference=@n20.Custom Map Widget))
   > n22 = CreateWidgetAsync!(
    execute:Exec
    UserWidgetSoftClass:User Widget Soft Class Reference=@n20.Custom Map Widget
    OwningPlayer:Player Controller Object Reference
    bSuspendInputUntilComplete:bool=true
   ) -> (then:Exec, OnComplete:Exec, OnFailure:Exec, UserWidget:User Widget Object Reference) {is=title K2Node_AsyncAction}
   ClearChildren(execute=@n22.OnComplete, self=Get CustomWidget_Slot() -> (CustomWidget_Slot:Named Slot Object Reference))
   > AddChild(self=Get CustomWidget_Slot(), Content=@n22)
   > SetActiveWidgetIndex(self=Get Switcher_HoveredDetail(), Index=2)
   ClearChildren(execute=@n21.else, self=Get CustomWidget_Slot())
   n23 = OnEntryInitialized!() -> (OutputDelegate, then, Item:Object Reference, Widget:User Widget Object Reference) {is=title K2Node_CustomEvent}
   > n24 = Cast To WBP_QuestCategoryRow_Entry!(execute, Object=@n23.Widget) -> (then, CastFailed, AsWBP Quest Category Row Entry:WBP Quest Category Row Entry Object Reference) {is=title K2Node_DynamicCast}
   n25 = Create Event()
   > Bind Event to On Entry Clicked(self:WBP Quest Category Row Entry Object Reference=@n24, Delegate=@n25) {is=title K2Node_AddDelegate}
   n26 = Create Event()
   > Bind Event to On Entry Hovered(self:WBP Quest Category Row Entry Object Reference=@n24, Delegate=@n26) {is=title K2Node_AddDelegate}
   n27 = Cast To WBP_QuestTile_Base!(execute=@n24.CastFailed, Object=@n23.Widget) -> (then, CastFailed, AsWBP Quest Tile Base:WBP Quest Tile Base Object Reference) {is=title K2Node_DynamicCast}
   > Bind Event to On Hovered(self:Common Button Base Object Reference=@n27, Delegate=@n26) {is=title K2Node_AddDelegate}
   n28 = OnEntryReleased!() -> (OutputDelegate, then, Widget:User Widget Object Reference) {is=title K2Node_CustomEvent}
   > n29 = Cast To WBP_QuestCategoryRow_Entry(Object=@n28.Widget)
   > Unbind Event from On Entry Clicked(self:WBP Quest Category Row Entry Object Reference=@n29, Delegate=@n25) {is=title K2Node_RemoveDelegate}
   > Unbind Event from On Entry Hovered(self:WBP Quest Category Row Entry Object Reference=@n29, Delegate=@n26) {is=title K2Node_RemoveDelegate}
   n30 = Cast To WBP_QuestTile_Base(execute=@n29.CastFailed, Object=@n28.Widget)
   > Unbind Event from On Hovered(self:Common Button Base Object Reference=@n30, Delegate=@n26) {is=title K2Node_RemoveDelegate}
   n31 = Event PreConstruct()
   > n32 = Set Is Design Time(Is Design Time:bool=@n31.IsDesignTime) -> (Output_Get:bool)
   > n33 = Branch(Condition=@n32)
   n34 = Sequence!(execute=[@n39.then, @n33.then]) -> (then_0, then_1, then_2:Exec)
   > n35 = Construct Object from Class!(
    execute
    Class=/QuestUI/QuestScreen/Debug/DebugQuestUIHeaderVM.DebugQuestUIHeaderVM_C
    self
    IsGameModeTitle:bool=true
    Title:Text="LOCTEXT(\"Header\")"
   ) -> (then, ReturnValue:Debug Quest UIHeader VM Object Reference)
   > Set HeaderVM(HeaderVM:Debug Quest UIHeader VM Object Reference=@n35) -> (Output_Get:Debug Quest UIHeader VM Object Reference)
   > n36 = Construct Object from Class(
    Class=&DebugQuestUIHeaderVM_C
    IsGameModeTitle=False
    Title="LOCTEXT(\"Not Available in the current mode\")"
   )
   > Set SeparatorVM(SeparatorVM:Debug Quest UIHeader VM Object Reference=@n36) -> (Output_Get:Debug Quest UIHeader VM Object Reference)
   > Set with Broadcast Icon!(
    execute
    Icon:Material Instance Soft Object Reference=/FortUILibrary/Textures/MaterialInstances/MI_UIKit_Icon_Locked
    self:Debug Quest UIHeader VM Object Reference=Get SeparatorVM() -> (SeparatorVM:Debug Quest UIHeader VM Object Reference)
   ) -> (then, Output_Get:Material Instance Soft Object Reference)
   > n37 = Construct Object from Class!(execute, Class=/QuestUI/QuestScreen/Modal/QuestUISpacerVM.QuestUISpacerVM_C, self, Size:double=66) -> (then, ReturnValue:Quest UISpacer VM Object Reference)
   > Set SpacerVM(SpacerVM:Quest UISpacer VM Object Reference=@n37) -> (Output_Get:Quest UISpacer VM Object Reference)
   n38 = Get ListView(self=Get QuestCategoriesList())
   Bind Event to On Entry Initialized(execute=@n34.then_1, self:List View Object Reference=@n38, Delegate=Create Event()) {is=title K2Node_AddDelegate}
   > Bind Event to On Entry Released(self:List View Base Object Reference=@n38, Delegate=Create Event()) {is=title K2Node_AddDelegate}
   n39 = SetDebugText!(execute=@n33.else, self:WBP UIKit Frontend Layout Object Reference=Get WBP_UIKit_FrontendLayout(), InText:Text, isWarning:bool=false) -> (then)
   Event On Deactivated()
   n40 = Get QuestModalRef()
   > n41 = Branch(Condition=IsActivated!(self:Common Activatable Widget Object Reference=@n40) -> (ReturnValue:bool))
   > n42 = DeactivateWidget(self=@n40)
   Unbind Event from On Widget Deactivated(execute=[@n42.then, @n41.else], self:Common Activatable Widget Object Reference=@n40, Delegate=@n7) {is=title K2Node_RemoveDelegate}
   > RemoveFromParent(self=@n40)
   > n43 = Set QuestModalRef(QuestModalRef:WBP Quests Modal Iteration 2 Object Reference) -> (Output_Get:WBP Quests Modal Iteration 2 Object Reference)
   n44 = CreateWidgetAsync(
    execute=@n55.then
    UserWidgetSoftClass=/QuestUI/QuestScreen/Modal/WBP_QuestsModalIteration2.WBP_QuestsModalIteration2_C
    bSuspendInputUntilComplete=true
   )
   n45 = Cast To WBP_QuestsModalIteration2!(execute=@n68.then_0, Object=@n44) -> (then, CastFailed, AsWBP Quests Modal Iteration 2:WBP Quests Modal Iteration 2 Object Reference) {is=title K2Node_DynamicCast}
   n46 = Get SideNavigation!(self:WBP Quest Modal Side Navigation Object Reference=Get WBP_QuestModal_SideNavigation!(self:WBP Quests Modal Iteration 2 Object Reference=@n45) -> (WBP_QuestModal_SideNavigation:WBP Quest Modal Side Navigation Object Reference)) -> (SideNavigation:WBP UIKit Side Navigation Object Reference)
   Unbind Event from On Side Navigation Selection Changed(execute=@n43.then, self:WBP UIKit Side Navigation Base Object Reference=@n46, Delegate=@n8) {is=title K2Node_RemoveDelegate}
   > n47 = Branch(Condition=Get MapContainerDeactivating() -> (MapContainerDeactivating:bool))
   > CancelClearSelectionTimer() {is=title K2Node_CallFunction}
   > n48 = Set MapContainerDeactivating(MapContainerDeactivating:bool=false) -> (Output_Get:bool)
   n49 = Create Event()
   n50 = Get CommonInputSubsystem()
   Unbind Event from On Input Method Changed(execute=[@n48.then, @n52.then], self=@n50, Delegate=@n49)
   > BindSidebarEvents!(execute, self, Bind:bool=false) -> (then) {is=title K2Node_CallFunction}
   n51 = Get CinematicListener() -> (CinematicListener:Async Action Start Listening To Event Object Reference)
   > Is Valid(InputObject=@n51)
   > StopListeningToEvent(self:Async Action Start Listening To Event Object Reference=@n51)
   n52 = ClearSelectionData(execute=@n47.else, self:Self Object Reference)
   Event Construct()
   n53 = IsInZone!(self:Fort Global UIContext Object Reference=Get FortGlobalUIContext() -> (ReturnValue:Fort Global UIContext Object Reference) {is=title K2Node_GetSubsystem}) -> (ReturnValue:bool)
   > SetVisibility(
    self=Get SeasonPass_Level() -> (SeasonPass_Level:Season Pass Level Object Reference)
    InVisibility=Select(Option 0:ESlateVisibility Enum=SelfHitTestInvisible, Option 1:ESlateVisibility Enum=Collapsed, Index=@n53) -> (ReturnValue:ESlateVisibility Enum)
   )
   > SetSafezoneTopPadding!(
    execute
    self:WBP UIKit Frontend Layout Object Reference=Get WBP_UIKit_FrontendLayout()
    Safezone_FrontendTopPadding:double=Get FrontendSpacer() -> (FrontendSpacer:double)
    SafeZone_InGameTopPadding:double=Get InGameSpacer() -> (InGameSpacer:double)
    bUseCollapsedSafezoneTopPadding:bool=@n53
   ) -> (then)
   > SetBottomButtonVisibility!(
    execute
    self:WBP Bottom Bar Decoy Button Object Reference=Get DecoyButton_DebugQuestUI() -> (DecoyButton_DebugQuestUI:WBP Bottom Bar Decoy Button Object Reference)
    ShouldButtonBeVisible:bool="Not Equal Exactly (String)"(A=GetBuildConfiguration(), B=Get ShippingConfig())
   ) -> (then)
   n54 = Event OnInitialized()
   SetContentPanelSizeX!(
    execute=[@n54.then, @n56.OutputPin]
    self:WBP UIKit Frontend Layout Object Reference=Get WBP_UIKit_FrontendLayout()
    ContentPanelPrimarySizeX:double=Get ContentPanelPrimaryWidthOverride() -> (ContentPanelPrimaryWidthOverride:double)
   ) -> (then)
   > SetCompactUIStyle() {is=title K2Node_CallFunction}
   Event Handle Map Screen Container Deactivating() {is=title K2Node_Event}
   > Set MapContainerDeactivating(MapContainerDeactivating=true)
   n55 = Event On Activated()
   n56 = Reroute Node()
   Call On Category Screen Created!(execute:Exec=@n67.then, self:WBP Quest Screen Object Reference) -> (then:Exec) {is=title K2Node_CallDelegate}
   n57 = For Each Loop with Break(Exec=[@n69.else, @n60.CastFailed, @n59.else], Array=Get Category Tabs VM(), Break=@n70.then)
   n58 = Get ActiveQuestProductVM()
   PushQuestModal(execute=@n57.Completed, CategoryInterfaceVM=@n57, ContentVM=@n58, IsRestoringSelection=false)
   n59 = Get!(execute:Exec=@n65.then) -> (then:Exec, else:Exec, SelectedCategory:MVVM Base Viewmodel Object Reference)
   > n60 = Cast To QuestUICategoryVM!(execute, Object=@n59) -> (then, CastFailed, AsQuest UI Category View Model:Quest UI Category View Model Object Reference) {is=title K2Node_DynamicCast}
   PushQuestModal(execute=@n69.then, CategoryInterfaceVM=@n60, ContentVM=@n58, IsRestoringSelection=false)
   OnQuestCinematicStarted(execute=@n63.OnEventReceived)
   n61 = Set QuestModalRef(execute=@n45.then, QuestModalRef=@n45)
   > Bind Event to On Widget Deactivated(self:Common Activatable Widget Object Reference=@n61, Delegate=@n7) {is=title K2Node_AddDelegate}
   > AddChild(self=Get WidgetSwitcher(), Content=Get QuestModalRef())
   > Bind Event to On Side Navigation Selection Changed(self:WBP UIKit Side Navigation Base Object Reference=@n46, Delegate=@n8) {is=title K2Node_AddDelegate}
   > FireQuestAnalytics_MainViewOpened!(execute, QuestScreen:Quest UIScreen Object Reference=Self-Reference()) -> (then)
   > FireQuestAnalytics_SubViewOpened!(
    execute
    PlayerController:Player Controller Object Reference=GetOwningPlayer()
    ProductVM:Quest UI Product View Model Object Reference=Get ActiveQuestProductVM()
    CategoryInterfaceVM:Quest UI Category Interface View Model Object Reference
   ) -> (then)
   > Assign On Input Method Changed!(execute:Exec, self:Common Input Object Reference=@n50, Delegate:"Delegate (by ref)"=@n49) -> (then:Exec) {is=title K2Node_AssignDelegate}
   > InitSidebar()
   > BindSidebarEvents(Bind=true)
   > UpdateClearBadgesVisibility!(execute, self, InputType:ECommonInputType Enum=GetCurrentInputType(self=Get CommonInputSubsystem())) -> (then)
   > RefreshFocus()
   > RestoreSelection()
   > n62 = GetEventRouter!(
    execute
    Scope:Object Class Reference=/Script/FortniteGame.FortPlayspace
    ContextActor:Actor Object Reference=GetOwningPlayer()
    None:%=Get GameplayEventRouterSubsystem() -> (ReturnValue:Gameplay Event Router Subsystem Object Reference) {is=title K2Node_GetSubsystem}
   ) -> (then, ReturnValue:Gameplay Event Router Component Object Reference)
   > n63 = StartListeningToEvent!(
    execute:Exec
    Target:Gameplay Event Router Component Object Reference=@n62
    EventType:Script Struct Object Reference=/QuestUI/QuestScreen/RewardsPanel/QuestScreenCinematicEvent
    Context:Object Reference
    EventBubblingRule:EEventBubblingRule Enum
    None:%=Self-Reference()
   ) -> (
    then:Exec
    AsyncTaskProxy:Async Action Start Listening To Event Object Reference
    OnEventReceived:Exec
    Payload:Quest Screen Cinematic Event Structure
   )
   > Set CinematicListener(CinematicListener:Async Action Start Listening To Event Object Reference=@n63) -> (Output_Get:Async Action Start Listening To Event Object Reference)
   > n64 = Branch(Condition=IsInZone(self=Get FortGlobalUIContext()))
   n65 = SetDisableTransitionAnimation(execute=[@n66.then, @n64.then], self=Get WidgetSwitcher(), bDisableAnimation=true)
   n66 = Branch(execute=@n64.else, Condition=GetConfigDataBool!(EntryName:String=quest_category_tab_ui_optimization) -> (bValueOut:bool, ReturnValue:bool))
   n67 = DelayUntilNextTick!(execute=@n68.then_1) -> (then)
   n68 = Sequence!(execute=@n44.OnComplete) -> (then_0, then_1)
   n69 = Branch(execute=@n60.then, Condition=Get bContentEnabled(self=@n60))
   n70 = Branch(execute=@n57.LoopBody, Condition=Get bContentEnabled(self=@n57))
  ]}
 HandleActiveProductCategoriesChanged {nodes [
   HandleActiveProductCategoriesChanged() {is=title K2Node_FunctionEntry}
   > RefreshQuestList!(execute, self, VM:Quest UI Product View Model Object Reference=Get ActiveQuestProductVM()) -> (then)
  ]}
 HandleSidebarClearBadgesClicked {nodes [
   HandleSidebarClearBadgesClicked() {is=title K2Node_FunctionEntry}
   > MarkSeen!(execute, self:Quest UI Base View Model Object Reference=Get QuestUIPlayerVM(), bCommitToMcp:bool=true) -> (then)
   > FireQuestAnalytics_ElementInteracted!(
    execute
    PlayerController:Player Controller Object Reference=GetOwningPlayer()
    Action:"String (by ref)"=BR_Quests_Button_ClearBadges
    ContentVM:Quest UI Base Content View Model Object Reference
   ) -> (then)
  ]}
 HandleSidebarCollapsedStateChanged {nodes [
   n1 = HandleSidebarCollapsedStateChanged!() -> (then, Collapsed:bool=~) {is=title K2Node_FunctionEntry}
   > PlayAnimation(
    InAnimation=Get HideRewards() -> (HideRewards:Widget Animation Object Reference)
    PlayMode=Select(Option 0:EUMGSequencePlayMode Enum, Option 1:EUMGSequencePlayMode Enum=Reverse, Index=@n1) -> (ReturnValue:EUMGSequencePlayMode Enum)
   )
  ]}
 HandleSidebarProductClicked {nodes [
   HandleSidebarProductClicked() {is=title K2Node_FunctionEntry}
   n1 = Get CommonInputSubsystem()
   > Branch(Condition="Equal (Enum)"(A=GetCurrentInputType(self=@n1), B=Gamepad))
   > RefreshFocus()
  ]}
 HandleSidebarProductSelected {nodes [
   n1 = HandleSidebarProductSelected!() -> (then, ProductVM:Quest UI Product View Model Object Reference) {is=title K2Node_FunctionEntry}
   > SetActiveProductVM!(execute, self, ProductVM:Quest UI Product View Model Object Reference=@n1) -> (then)
  ]}
 HandleSidebarRequiresVisibilityChanged {nodes [
   n1 = HandleSidebarRequiresVisibilityChanged!() -> (then, RequiresVisibility:bool=~) {is=title K2Node_FunctionEntry}
   > SetSideNavVisibility!(execute, self:WBP UIKit Frontend Layout Object Reference=Get WBP_UIKit_FrontendLayout(), IsVisible:bool=@n1) -> (then)
  ]}
 HasNewContent {nodes [
   n1 = HasNewContent!() -> (then, HasNewContent:bool=~) {is=title K2Node_FunctionEntry}
   > SetIsLocked!(
    execute
    self:Common Button Base Object Reference=Get Button_ClearBadges_Gamepad() -> (Button_ClearBadges_Gamepad:WBP UIKit Button Quiet Object Reference)
    bInIsLocked:bool=NOT Boolean(A=@n1)
   ) -> (then)
   > SetRenderOpacity(
    self=Get WBP_ActionBarContainer_Left() -> (WBP_ActionBarContainer_Left:WBP Action Bar Container Object Reference)
    InOpacity=SelectFloat(A=1, B=0.4, bPickA=@n1)
   )
  ]}
 InitSidebar {nodes [
   InitSidebar() {is=title K2Node_FunctionEntry}
   n1 = Get SideNavigation!() -> (SideNavigation:WBP Quest Screen Side Navigation Object Reference)
   > SetIsCollapsed!(execute, self:WBP Quest Screen Side Navigation Object Reference=@n1, IsCollapsed:bool=true, SkipAnimation:bool=true) -> (then)
   > SetActiveProductVM(ProductVM=GetSelectedProductVM!(self:WBP Quest Screen Side Navigation Object Reference=@n1) -> (ProductVM:Quest UI Product View Model Object Reference))
   > HandleSidebarRequiresVisibilityChanged!(
    execute
    self
    RequiresVisibility:bool=GetRequiresVisibility!(self:WBP Quest Screen Side Navigation Object Reference=@n1) -> (RequiresVisibility:bool)
   ) -> (then) {is=title K2Node_CallFunction}
  ]}
 InputChanged {nodes [
   n1 = InputChanged!() -> (then, InputType:ECommonInputType Enum) {is=title K2Node_FunctionEntry}
   > UpdateClearBadgesVisibility(InputType=@n1)
   > RefreshFocus()
  ]}
 OnCategoryScreenCreated {nodes [OnCategoryScreenCreated() {is=title K2Node_FunctionEntry}]}
 OnQuestCinematicStarted {nodes [
   OnQuestCinematicStarted() {is=title K2Node_FunctionEntry}
   > Set CinematicPlaying(CinematicPlaying:bool=true) -> (Output_Get:bool)
  ]}
 OnShouldDisplayBackActionInActionBar {nodes [
   OnShouldDisplayBackActionInActionBar() {is=title K2Node_FunctionEntry}
   > Return Node!(execute, bOutShouldDisplayBackActionInActionBar:bool=true, ReturnValue:bool=true)
  ]}
 OnShouldShowTabOnMapScreenForMinigame {nodes [
   OnShouldShowTabOnMapScreenForMinigame!() -> (then, Minigame:Fort Minigame Object Reference) {is=title K2Node_FunctionEntry}
   > n1 = Cast To FrontEnd_PlayerController!(execute, Object=GetOwningPlayer()) -> (then, CastFailed, AsFront End Player Controller:Front End Player Controller Object Reference) {is=title K2Node_DynamicCast}
   > Return Node!(execute, bOutShouldShowTabOnMapScreenForMinigame:bool=false, ReturnValue=true)
   n2 = Get Minigame() -> (Minigame:Fort Minigame Object Reference)
   n3 = Is Valid(exec=@n1.CastFailed, InputObject=@n2)
   Return Node(execute=@n3.Is Not Valid, bOutShouldShowTabOnMapScreenForMinigame=true, ReturnValue=true)
   > Return Node(bOutShouldShowTabOnMapScreenForMinigame=HasStarted!(self:Fort Minigame Object Reference=@n2) -> (ReturnValue:bool), ReturnValue=true)
  ]}
 PushQuestModal {nodes [
   n1 = PushQuestModal!() -> (
    then
    CategoryInterfaceVM:Quest UI Category Interface View Model Object Reference
    ContentVM:Quest UI Base Content View Model Object Reference
    IsRestoringSelection:bool=~
   ) {is=title K2Node_FunctionEntry}
   > SetIsRestoringSelection!(
    execute
    self:WBP Quests Modal Iteration 2 Object Reference=Get QuestModalRef()
    IsRestoringSelection:bool=Get IsRestoringSelection() -> (IsRestoringSelection:bool)
   ) -> (then)
   > CancelClearSelectionTimer()
   n2 = Get QuestModalRef()
   > SetActiveWidget(self=Get WidgetSwitcher(), Widget=@n2)
   > UpdateQuestTabs()
   > SetActiveTab!(
    execute
    self:WBP Quests Modal Iteration 2 Object Reference=@n2
    CategoryInterfaceVM:Quest UI Category Interface View Model Object Reference=@n1
    ContentVM:Quest UI Base Content View Model Object Reference=@n1.ContentVM
   ) -> (then)
   > Unbind Event from On Input Method Changed(self=Get CommonInputSubsystem(), Delegate=Create Event())
  ]}
 QuestModalDeactivated {nodes [
   QuestModalDeactivated() {is=title K2Node_FunctionEntry}
   > Bind Event to On Input Method Changed(self=Get CommonInputSubsystem(), Delegate=Create Event())
   > RefreshFocus()
   > n1 = Branch(Condition=IsActivated())
   > n2 = FireQuestAnalytics_SubViewOpened(PlayerController=GetOwningPlayer(), ProductVM=Get ActiveQuestProductVM())
   n3 = Get!(execute=[@n2.then, @n1.else]) -> (then, else, QuestModalRef:WBP Quests Modal Iteration 2 Object Reference)
   n4 = Get IsExiting!(self:WBP Quests Modal Iteration 2 Object Reference=@n3) -> (IsExiting:bool)
   > n5 = Branch(Condition=@n4)
   Call On Quest Modal Deactivated!(execute:Exec=[@n6.then, @n5.then], self:Quest UIScreen Object Reference, bWantsExit:bool=@n4) -> (then:Exec) {is=title K2Node_CallDelegate}
   n6 = StartClearSelectionTimer(execute=@n5.else)
  ]}
 RefreshFocus {nodes [
   RefreshFocus() {is=title K2Node_FunctionEntry}
   > RequestRefreshFocus(self:Common Activatable Widget Object Reference)
  ]}
 RefreshQuestList {nodes [
   n1 = RefreshQuestList!() -> (then, VM:Quest UI Product View Model Object Reference) {is=title K2Node_FunctionEntry}
   > Is Valid(InputObject=@n1)
   > Clear(TargetArray:Array of Object References=Get L_CategoriesList() -> (L_CategoriesList:Array of Object References))
   > n2 = GetConsoleVariableBoolValue!(execute, VariableName:String=QuestUI.AllowNotInstalledMessaging) -> (then, ReturnValue:bool)
   n3 = Get ProductInstallState!(self:Quest UI Product View Model Object Reference=@n1) -> (ProductInstallState:EVerseURIInstallState Enum=Installed)
   > n4 = Branch(Condition=OR Boolean(A=NOT Boolean(A=@n2), B="Equal (Enum)"(A:EVerseURIInstallState Enum=@n3, B:EVerseURIInstallState Enum=Installed)))
   > n5 = Branch(Condition=IsInZone(self=Get FortGlobalUIContext()))
   n6 = Sequence!(execute=@n5.else) -> (then_0, then_1, then_2:Exec, then_3:Exec)
   n7 = Get HeaderVM() -> (HeaderVM:Debug Quest UIHeader VM Object Reference)
   Set with Broadcast Icon(
    execute=@n6.then_2
    Icon=Get ProductIcon!(self:Quest UI Product View Model Object Reference=Get VM() -> (VM:Quest UI Product View Model Object Reference)) -> (ProductIcon:Material Instance Soft Object Reference)
    self=@n7
   )
   > Set with Broadcast Title!(
    execute
    Title:Text=Get ProductName!(self:Quest UI Product View Model Object Reference=Get VM()) -> (ProductName:Text)
    self:Debug Quest UIHeader VM Object Reference=@n7
   ) -> (then, Output_Get:Text)
   > Add(TargetArray:Array of Object References=Get L_CategoriesList(), NewItem:"Object Reference (by ref)"=@n7) {is=title K2Node_CallArrayFunction}
   > Append Array!(
    execute
    TargetArray:Array of Object References=Get L_CategoriesList()
    SourceArray:Array of Object References=Get OrderedAvailableRowVMs!(self:Quest UI Product View Model Object Reference=Get VM()) -> (OrderedAvailableRowVMs:Array of Quest UI Category Row View Model Object References)
   ) -> (then)
   n8 = Get OrderedUnavailableRowVMs!(self:Quest UI Product View Model Object Reference=Get VM()) -> (OrderedUnavailableRowVMs:Array of Quest UI Category Row View Model Object References)
   > n9 = Branch(Condition=Is Empty!(TargetArray:Array of Quest UI Category Row View Model Object References=@n8) -> (ReturnValue:bool))
   Add(execute=[@n10.then, @n9.then], TargetArray=Get L_CategoriesList(), NewItem=Get SpacerVM() -> (SpacerVM:Quest UISpacer VM Object Reference))
   > SetListItems!(
    execute
    self:Fort Polymorphic View Object Reference=Get QuestCategoriesList()
    InListItems:Array of Object References=Get L_CategoriesList()
   ) -> (then)
   Add(execute=@n9.else, TargetArray=Get L_CategoriesList(), NewItem=Get SeparatorVM())
   > n10 = Append Array(TargetArray=Get L_CategoriesList(), SourceArray=@n8)
   n11 = Get VM()
   n12 = Get OrderedUnavailableRowVMs(self=@n11)
   n13 = Get OrderedAvailableRowVMs(self=@n11)
   n14 = Is Empty(TargetArray=@n13)
   n15 = Branch(execute=@n6.then_3, Condition=AND Boolean(A=@n14, B=Is Empty(TargetArray=@n12)))
   Clear(
    execute=[@n17.CastFailed, @n15.then]
    TargetArray:Array of Quest UI Reward VM Object References=Get CategoryRewardVMs!() -> (CategoryRewardVMs)
   )
   > UpdateRewards(self=Get WBP_SideRewardsPanel())
   > SetVisibility(self=Get WBP_SideRewardsPanel(), InVisibility=SelfHitTestInvisible)
   n16 = "Get (a copy)"(Array:Array of Quest UI Category Row View Model Object References=Select(
     Option 0:Array of Quest UI Category Row View Model Object References=@n13
     Option 1:Array of Quest UI Category Row View Model Object References=@n12
     Index=@n14
    ) -> (ReturnValue:Array of Quest UI Category Row View Model Object References)) -> (Output:Quest UI Category Row View Model Object Reference)
   n17 = Cast To QuestUICategoryVM(
    execute=@n15.else
    Object="Get (a copy)"(Array:Array of Quest UI Category Interface View Model Object References=Get OrderedCategoryInterfaceVMs!(self:Quest UI Category Row View Model Object Reference=@n16) -> (OrderedCategoryInterfaceVMs:Array of Quest UI Category Interface View Model Object References)) -> (Output:Quest UI Category Interface View Model Object Reference)
   )
   n18 = Get CustomMapWidget(self=@n17)
   > UpdateCustomWidget(Custom Map Widget=@n18) {is=title K2Node_CallFunction}
   > n19 = Branch(Condition=IsValidSoftClassReference(SoftClassReference=@n18))
   n20 = Set CategoryRewardVMs(execute=@n19.else, CategoryRewardVMs=Get CategoryRewardVMs!(self:Quest UI Category View Model Object Reference=@n17) -> (CategoryRewardVMs))
   n21 = Get WBP_SideRewardsPanel()
   > UpdateRewards(self=@n21, RewardsCustomWidgets=@n20, QuestContentVM=@n17)
   > SetSelectedPage(self=Get Pages(self=@n21), SelectedOption=0)
   SetActiveWidgetIndex(execute=@n6.then_0, self:Widget Switcher Object Reference=Get ContentSwitcher(), Index=0)
   Add(
    execute=@n6.then_1
    TargetArray=Get L_CategoriesList()
    NewItem=Get TrackerVM!(self:Quest UI Player View Model Object Reference=Get QuestUIPlayerVM()) -> (TrackerVM:Quest UI Tracker View Model Object Reference)
   )
   SetActiveWidgetIndex(execute=@n4.else, self=Get ContentSwitcher(), Index=1)
   > SetInstallState!(
    execute
    self:WBP Product Install Error Object Reference=Get ProductInstallError() -> (ProductInstallError:WBP Product Install Error Object Reference)
    ProductVM:Quest UI Product View Model Object Reference=Get VM()
   ) -> (then)
   > UpdateRewards(self=Get WBP_SideRewardsPanel())
  ]}
 RestoreSelection {nodes [
   RestoreSelection() {is=title K2Node_FunctionEntry}
   n1 = Get QuestUIPlayerVM()
   > Is Valid(InputObject=@n1)
   > n2 = Branch(Condition=HasSelectionData!(self:Quest UI Player View Model Object Reference=@n1) -> (ReturnValue:bool))
   > n3 = ConsumeSelectionData!(execute, self:Quest UI Player View Model Object Reference=@n1) -> (then, ReturnValue:Quest UISelection Data Structure)
   > n4 = Sequence!(execute) -> (then_0, then_1, then_2)
   n5 = Break Quest UISelection Data!(QuestUISelectionData:"Quest UISelection Data Structure (by ref)"=@n3) -> (
    ProductVM:Quest UI Product View Model Object Reference
    CategoryInterfaceVM:Quest UI Category Interface View Model Object Reference
    ContentVM:Quest UI Base Content View Model Object Reference
   ) {is=title K2Node_BreakStruct}
   > Is Valid(InputObject=@n5)
   > SetActiveProductVM(ProductVM=@n5)
   n6 = Find Item!(
    TargetArray:Array of Quest UI Product View Model Object References=Get OrderedProductVMs!(self:Quest UI Player View Model Object Reference=Get QuestUIPlayerVM()) -> (OrderedProductVMs:Array of Quest UI Product View Model Object References)
    ItemToFind:"Quest UI Product View Model Object Reference (by ref)"=@n5
   ) -> (ReturnValue)
   > Branch(Condition="integer >= integer"(A=@n6))
   > SelectTabByIndex!(
    execute
    self:UIKit Tab Group Object Reference=Get SideNavigation!(self:WBP Quest Screen Side Navigation Object Reference=Get SideNavigation!() -> (SideNavigation:WBP Quest Screen Side Navigation Object Reference)) -> (SideNavigation:WBP UIKit Side Navigation Object Reference)
    Index:int=@n6
    bSuppressClickFeedback:bool=true
   ) -> (then, ReturnValue:bool)
   Is Valid(exec=@n4.then_1, InputObject=@n5.CategoryInterfaceVM)
   > PushQuestModal(CategoryInterfaceVM=@n5.CategoryInterfaceVM, ContentVM=@n5.ContentVM, IsRestoringSelection=true) {is=title K2Node_CallFunction}
   Set CinematicPlaying(execute=@n4.then_2, CinematicPlaying=false)
   n7 = Cast To FortPlayerControllerFrontEnd!(execute=@n2.else, Object=GetOwningPlayer()) -> (then, CastFailed, AsFort Player Controller Front End:Fort Player Controller Front End Object Reference) {is=title K2Node_DynamicCast}
   n8 = Get TrackerVM(self=Get QuestUIPlayerVM())
   Is Valid(exec=@n7.CastFailed, InputObject=@n8)
   > Branch(Condition="integer > integer"(A=GetCurrentTrackedQuestCount!(self:Quest UI Tracker View Model Object Reference=@n8) -> (ReturnValue:int=0)))
   > PushQuestModal(CategoryInterfaceVM=@n8, IsRestoringSelection=false)
  ]}
 SetActiveProductVM {nodes [
   SetActiveProductVM!() -> (then, ProductVM:Quest UI Product View Model Object Reference) {is=title K2Node_FunctionEntry}
   n1 = Get ActiveQuestProductVM()
   > n2 = Is Valid(InputObject=@n1)
   > n3 = Unbind Event from On Product Categories Changed(self:Quest UI Product View Model Object Reference=@n1, Delegate=Create Event()) {is=title K2Node_RemoveDelegate}
   n4 = Set ActiveQuestProductVM(
    execute=[@n3.then, @n2.Is Not Valid]
    ActiveQuestProductVM:Quest UI Product View Model Object Reference=Get ProductVM() -> (ProductVM:Quest UI Product View Model Object Reference)
   ) -> (Output_Get:Quest UI Product View Model Object Reference)
   > n5 = Is Valid(InputObject=@n4)
   > Bind Event to On Product Categories Changed(self:Quest UI Product View Model Object Reference=@n4, Delegate=Create Event()) {is=title K2Node_AddDelegate}
   > n6 = Branch(Condition=IsActivated())
   > n7 = FireQuestAnalytics_SubViewOpened(PlayerController=GetOwningPlayer(), ProductVM=Get ActiveQuestProductVM())
   RefreshQuestList(execute=[@n7.then, @n6.else, @n5.Is Not Valid], VM=Get ActiveQuestProductVM()) {is=title K2Node_CallFunction}
   > UpdateQuestTabs()
   n8 = Get ProductVM()
   > Is Valid(InputObject=@n8)
   > n9 = Get()
   > SetQuestUIProductVM!(execute, self:WBP Quests Modal Iteration 2 Object Reference=@n9, Viewmodel:Quest UI Product View Model Object Reference=@n8) -> (then)
  ]}
 SetCompactUIStyle {nodes [
   SetCompactUIStyle() {is=title K2Node_FunctionEntry}
   > Branch(Condition=IsMobileGame())
   > SetCompactUIStyle!(
    execute
    self:WBP UIKit Frontend Layout Object Reference=Get WBP_UIKit_FrontendLayout()
    ContentPanelSizeX:double=Get ContentPanelPrimaryWidthOverride_CompactUI() -> (ContentPanelPrimaryWidthOverride_CompactUI:double)
    TopBarSpacer:double=Get TopBarSpacerHeight_CompactUI() -> (TopBarSpacerHeight_CompactUI:double)
    SideNavMaxSizeY:double=Get SideNavMaxHeight_CompactUI() -> (SideNavMaxHeight_CompactUI:double)
    SafeZoneMarginY:double=Get SafeZoneMarginY_CompactUI() -> (SafeZoneMarginY_CompactUI:double)
    SafeZoneMarginX:double=Get SafeZoneMarginX_CompactUI() -> (SafeZoneMarginX_CompactUI:double)
   ) -> (then) {is=title K2Node_CallFunction}
   > Set FrontendSpacer(FrontendSpacer:double=Get FrontendSpacer_CompactUI() -> (FrontendSpacer_CompactUI:double)) -> (Output_Get:double)
   > Set InGameSpacer(InGameSpacer:double=Get InGameSpacer_CompactUI() -> (InGameSpacer_CompactUI:double)) -> (Output_Get:double)
   > Set SideNav_PaddingRight!(execute, SideNav_PaddingRight:double=0, self:WBP UIKit Frontend Layout Object Reference=Get WBP_UIKit_FrontendLayout()) -> (then, Output_Get:double)
   > DisplaySideNav(self:WBP UIKit Frontend Layout Object Reference=Get WBP_UIKit_FrontendLayout())
  ]}
 SetDebugPlaylistInfoIsWarning {nodes [
   n1 = SetDebugPlaylistInfoIsWarning!() -> (then, IsWarning:bool=~) {is=title K2Node_FunctionEntry}
   > SetDebugTextWarning!(execute, self:WBP UIKit Frontend Layout Object Reference=Get WBP_UIKit_FrontendLayout(), IsWarning:bool=@n1) -> (then)
  ]}
 SetDebugPlaylistInfoText {nodes [
   n1 = SetDebugPlaylistInfoText!() -> (then, InfoText:Text) {is=title K2Node_FunctionEntry}
   > SetDebugText(self=Get WBP_UIKit_FrontendLayout(), InText=@n1, isWarning=false)
  ]}
 StartClearSelectionTimer {nodes [
   StartClearSelectionTimer() {is=title K2Node_FunctionEntry}
   n1 = Get QuestUIPlayerVM()
   > Is Valid(InputObject=@n1)
   n2 = Get CinematicPlaying() -> (CinematicPlaying:bool)
   > Branch(Condition=AND Boolean(A=HasSelectionData(self=@n1), B=NOT Boolean(A=@n2)))
   > n3 = Set Timer by Event(Delegate=Create Event(), Time=Get ClearSelectionTime() -> (ClearSelectionTime:double), bLooping=false)
   > Set ClearSelectionTimerHandle(ClearSelectionTimerHandle:Timer Handle Structure=@n3) -> (Output_Get:Timer Handle Structure)
  ]}
 UpdateClearBadgesVisibility {nodes [
   n1 = UpdateClearBadgesVisibility!() -> (then, InputType:ECommonInputType Enum) {is=title K2Node_FunctionEntry}
   > SetVisibility(
    self=Get Button_ClearBadges_Gamepad()
    InVisibility=Select!(
     MouseAndKeyboard:ESlateVisibility Enum=Collapsed
     Gamepad:ESlateVisibility Enum=Visible
     Touch:ESlateVisibility Enum=Collapsed
     Count:ESlateVisibility Enum=Collapsed
     Index:ECommonInputType Enum=@n1
    ) -> (ReturnValue:ESlateVisibility Enum)
   )
   > SetClearBadgesVisibility!(
    execute
    self:WBP Quest Screen Side Navigation Object Reference=Get SideNavigation()
    Visibility:ESlateVisibility Enum=Select(MouseAndKeyboard=Visible, Gamepad=Collapsed, Touch=Visible, Count=Collapsed, Index=@n1)
   ) -> (then)
   > UpdateSideNavigationBottomPadding!(execute, self, Input Type:ECommonInputType Enum=Get InputType() -> (InputType:ECommonInputType Enum)) -> (then)
  ]}
 UpdateQuestTabs {nodes [
   UpdateQuestTabs() {is=title K2Node_FunctionEntry}
   > Clear(TargetArray:Array of Quest UI Category Interface View Model Object References=Get L_TabsVM() -> (L_TabsVM:Array of Quest UI Category Interface View Model Object References))
   > Add(
    TargetArray:Array of Quest UI Category Interface View Model Object References=Get L_TabsVM()
    NewItem:"Quest UI Category Interface View Model Object Reference (by ref)"=Get TrackerVM(self=Get QuestUIPlayerVM())
   )
   > n1 = Get!(execute) -> (then, else, ActiveQuestProductVM:Quest UI Product View Model Object Reference)
   > n2 = Set L_MakeArray(L_MakeArray:Array of Quest UI Category Row View Model Object References=Get OrderedAvailableRowVMs(self=@n1)) -> (Output_Get:Array of Quest UI Category Row View Model Object References)
   > Append Array(
    TargetArray:Array of Quest UI Category Row View Model Object References=@n2
    SourceArray:Array of Quest UI Category Row View Model Object References=Get OrderedUnavailableRowVMs(self=@n1)
   )
   > n3 = For Each Loop(Array:Array of Quest UI Category Row View Model Object References=Get L_MakeArray() -> (L_MakeArray:Array of Quest UI Category Row View Model Object References)) -> (Array Element:Quest UI Category Row View Model Object Reference)
   Set Category Tabs VM(execute=@n3.Completed, Category Tabs VM:Array of Quest UI Category Interface View Model Object References=Get L_TabsVM()) -> (Output_Get:Array of Quest UI Category Interface View Model Object References)
   > n4 = Get!(execute) -> (then, else, QuestModalRef:WBP Quests Modal Iteration 2 Object Reference)
   > PopulateCategories!(
    execute
    self:WBP Quests Modal Iteration 2 Object Reference=@n4
    Data:Array of Quest UI Category Interface View Model Object References=Get L_TabsVM()
   ) -> (then)
   n5 = For Each Loop(Exec=@n3.LoopBody, Array:Array of Quest UI Category Interface View Model Object References=Get OrderedCategoryInterfaceVMs(self=@n3)) -> (Array Element:Quest UI Category Interface View Model Object Reference)
   > Add(TargetArray=Get L_TabsVM(), NewItem=@n5)
  ]}
 UpdateSideNavigationBottomPadding {nodes [
   n1 = UpdateSideNavigationBottomPadding!() -> (then, Input Type:ECommonInputType Enum) {is=title K2Node_FunctionEntry}
   n2 = Get SideNavigation()
   n3 = Get!(self:User Widget Object Reference=@n2) -> (Padding_Left:float=~, Padding_Top:float=~, Padding_Right:float=~, Padding_Bottom:float=~)
   > Set!(
    execute
    Padding_Left:float=@n3
    Padding_Top:float=@n3.Padding_Top
    Padding_Right:float=@n3.Padding_Right
    Padding_Bottom:float=SelectFloat(A=100, B=0, bPickA="Equal (Enum)"(A:ECommonInputType Enum=@n1, B:ECommonInputType Enum=Gamepad))
    self:User Widget Object Reference=@n2
   ) -> (then, Output_Get:Margin Structure)
  ]}
]
variables [
 ActiveQuestProductVM {type=QuestUIProductVM, default=null}
 Category Tabs VM {type="array<QuestUICategoryInterfaceVM>", default []}
 CategoryRewardVMs {type="array<QuestUIRewardVM>", default []}
 CinematicListener {type=AsyncAction_StartListeningToEvent, default=null}
 CinematicPlaying {type=boolean}
 ClearSelectionTimerHandle {type=TimerHandle}
 ContentPanelPrimaryWidthOverride {type=number, default=1144}
 ContentPanelPrimaryWidthOverride_CompactUI {type=number, default=1144}
 FrontendSpacer {type=number, default=136}
 FrontendSpacer_CompactUI {type=number, default=160}
 HeaderVM {type=DebugQuestUIHeaderVM_C, default=null}
 InGameSpacer {type=number, default=32}
 InGameSpacer_CompactUI {type=number, default=48}
 Is Design Time {type=boolean}
 isShowingTracker {type=boolean}
 MapContainerDeactivating {type=boolean}
 OnCategoryScreenCreated {}
 QuestModalRef {type=WBP_QuestsModalIteration2_C, default=null}
 SafeZoneMarginX_CompactUI {type=number, default=48}
 SafeZoneMarginY_CompactUI {type=number, default=24}
 SelectedCategory {type=MVVMViewModelBase, default=null}
 SeparatorVM {type=DebugQuestUIHeaderVM_C, default=null}
 ShippingConfig {type=string, default=Shipping}
 SideNavMaxHeight_CompactUI {type=number}
 SpacerVM {type=QuestUISpacerVM_C, default=null}
 TopBarSpacerHeight_CompactUI {type=number, default=160}
 TrackedHeaderVM {type=DebugQuestUIHeaderVM_C, default=null}
]
widgets {
 info {inheritedWidgetCount=0, namedSlotCount=7, parentClass=/Script/QuestUI.QuestUIScreen, rootWidgetClass=&GridPanel, widgetCount=20}
 widgets [
  Root {widgetClassPath=&GridPanel, widget=Root, properties {columnFill [1], rowFill [1]}}
  Background {
   widgetClassPath=&FortLazyImage
   parent=Root
   widget=Background
   properties {
    bOverride_Cursor=true
    visibility=Visible
    brush {
     bIsDynamicallyLoaded=false
     drawAs=Image
     imageType=FullColor
     mirroring=NoMirror
     resourceName=None
     resourceObject=/QuestUI/QuestScreen/Materials/M_UI_TestBackground_Inst
     tiling=NoTile
     imageSize=<32 32>
     margin=<ltrb 0 0 0 0>
     outlineSettings {
      bUseBrushTransparency=false
      roundingType=HalfHeightRadius
      width=0
      color {colorUseRule=UseColor_Specified, specifiedColor=<rgba 0 0 0 0>}
      cornerRadii=<0 0 0 0>
     }
     tintColor {colorUseRule=UseColor_Specified, specifiedColor=<rgba 1 1 1 1>}
     uVRegion {bIsValid=false, max=<0 0>, min=<0 0>}
    }
   }
   slot {class=GridSlot}
  }
  WidgetSwitcher {
   widgetClassPath=&CommonActivatableWidgetSwitcher
   bIsVariable=true
   parent=Root
   widget=WidgetSwitcher
   slot {class=GridSlot, properties {layer=1}}
  }
  WBP_UIKit_FrontendLayout {
   widgetClassPath=/FortUILibrary/UIKit/Layout/WBP_UIKit_FrontendLayout
   bIsVariable=true
   parent=WidgetSwitcher
   widget=WBP_UIKit_FrontendLayout
   properties {bUsePostBufferUpdate=true, backgroundMaterial=&M_UI_TestBackground_Inst}
   slot {class=WidgetSwitcherSlot}
  }
  Overlay_110 {widgetClassPath=&Overlay, namedSlotHost=WBP_UIKit_FrontendLayout, widget=Overlay_110}
  WBP_ActionBarContainer_Left {
   widgetClassPath=/Game/UI/Frontend/BottomBar/WBP_ActionBarContainer
   bIsVariable=true
   parent=Overlay_110
   widget=WBP_ActionBarContainer_Left
   slot {class=OverlaySlot, properties {verticalAlignment=VAlign_Bottom}}
  }
  Button_Bar_Left {widgetClassPath=&HorizontalBox, namedSlotHost=WBP_ActionBarContainer_Left, widget=Button_Bar_Left}
  Button_ClearBadges_Gamepad {
   widgetClassPath=/FortUILibrary/UIKit/Buttons/WBP_UIKit_Button_Quiet
   bIsVariable=true
   parent=Button_Bar_Left
   widget=Button_ClearBadges_Gamepad
   properties {
    bHideInputActionWithKeyboard=true
    bIsFocusable=false
    text=Clear Badges
    triggeringInputAction {dataTable=&InputActionDataTable, rowName=GenericLeftStickClick}
   }
   slot {class=HorizontalBoxSlot}
  }
  Switcher_HoveredDetail {
   widgetClassPath=&CommonVisibilitySwitcher
   bIsVariable=true
   namedSlotHost=WBP_UIKit_FrontendLayout
   widget=Switcher_HoveredDetail
  }
  WBP_SideRewardsPanel {
   widgetClassPath=/QuestUI/QuestScreen/RewardsPanel/WBP_SideRewardsPanel
   bIsVariable=true
   parent=Switcher_HoveredDetail
   widget=WBP_SideRewardsPanel
   properties {bCanBeVisibleOnMobile=false, padding=<ltrb 64 0 0 0>}
   slot {class=CommonVisibilitySwitcherSlot, properties {verticalAlignment=VAlign_Center}}
  }
  WBP_CustomHoverContent_TrackedQuest {
   widgetClassPath=&/WBP_CustomHoverContent_TrackedQuest
   bIsVariable=true
   parent=Switcher_HoveredDetail
   widget=WBP_CustomHoverContent_TrackedQuest
   properties {bSupportsActivationFocus=false}
   slot {class=CommonVisibilitySwitcherSlot, properties {verticalAlignment=VAlign_Center, padding=<ltrb 64 0 0 0>}}
  }
  SB_CustomWidget {
   widgetClassPath=&SizeBox
   bIsVariable=true
   parent=Switcher_HoveredDetail
   widget=SB_CustomWidget
   properties {bOverride_HeightOverride=true, bOverride_WidthOverride=true, heightOverride=692.3077, widthOverride=653.8461}
   slot {class=CommonVisibilitySwitcherSlot, properties {horizontalAlignment=HAlign_Center, verticalAlignment=VAlign_Center}}
  }
  CustomWidget_Slot {
   widgetClassPath=&NamedSlot
   bIsVariable=true
   parent=SB_CustomWidget
   widget=CustomWidget_Slot
   slot {class=SizeBoxSlot, properties {padding=<ltrb 50 70 0 30>}}
  }
  ContentSwitcher {
   widgetClassPath=&CommonActivatableWidgetSwitcher
   bIsVariable=true
   namedSlotHost=WBP_UIKit_FrontendLayout
   widget=ContentSwitcher
  }
  QuestCategoriesList {
   widgetClassPath=&/WBP_QuestCategoriesList
   bIsVariable=true
   parent=ContentSwitcher
   widget=QuestCategoriesList
   properties {
    navigation=&self:WidgetTree.QuestCategoriesList.WidgetNavigation_0
    desiredFocusWidget {widgetName=None}
    padding=<ltrb 20 10 10 0>
   }
   slot {class=WidgetSwitcherSlot}
  }
  ProductInstallError {
   widgetClassPath=&/WBP_ProductInstallError
   bIsVariable=true
   parent=ContentSwitcher
   widget=ProductInstallError
   slot {class=WidgetSwitcherSlot}
  }
  Overlay_58 {widgetClassPath=&Overlay, namedSlotHost=WBP_UIKit_FrontendLayout, widget=Overlay_58}
  DecoyButton_DebugQuestUI {
   widgetClassPath=/Game/UI/Frontend/BottomBar/WBP_BottomBarDecoyButton
   bIsVariable=true
   parent=Overlay_58
   widget=DecoyButton_DebugQuestUI
   properties {
    holdData=&CommonUI_HoldData_Base_C
    triggeringInputAction {dataTable=/QuestUI/QuestScreen/RewardsPanel/QuestInputActionDataTable, rowName=DebugQuestUI}
   }
   slot {class=OverlaySlot}
  }
  SideNavigation {
   widgetClassPath=&/WBP_QuestScreen_SideNavigation
   bIsVariable=true
   namedSlotHost=WBP_UIKit_FrontendLayout
   widget=SideNavigation
  }
  SeasonPass_Level {
   widgetClassPath=/FortniteCoreUI/MultiProduct/SeasonPass_Level
   bIsVariable=true
   namedSlotHost=WBP_UIKit_FrontendLayout
   widget=SeasonPass_Level
  }
 ]
}
