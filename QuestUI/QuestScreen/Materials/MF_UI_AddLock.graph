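# UI material function that overlays a lock icon with a soft circular backing on an input RGBA image, likely for locked quest-screen imagery.
# The lock texture is placed by scale and anchor inputs, with an SDF circle supplying backing and inset masks.
# A softness value is divided by the viewport UI scale via a Custom node for DPI compensation.
# Lerps composite these into FinalColor and FinalOpacity, with inputs for locked crossfade and opacities.
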MaterialFunction /QuestUI/QuestScreen/Materials/MF_UI_AddLock
graph {
 properties {bExposeToLibrary=true, libraryCategoriesText [UI]}
 nodes [
  n1 = NamedRerouteUsage()
  n2 = OneMinus(None=NamedRerouteUsage())
  n3 = LinearInterpolate(B=NamedRerouteUsage(), Alpha=OneMinus(None=Multiply(A=@n1, B=@n2)))
  n4 = NamedRerouteUsage()
  n5 = NamedRerouteUsage()
  FunctionOutput(None=LinearInterpolate(A=NamedRerouteUsage(), B=LinearInterpolate(A=@n3, B=@n4, Alpha=@n5), Alpha=NamedRerouteUsage())) {outputName=FinalColor}
  n6 = Multiply(A=NamedRerouteUsage(), B=NamedRerouteUsage()) {constB=0.9}
  n7 = Max(A=Multiply(A=NamedRerouteUsage(), B=NamedRerouteUsage()), B=NamedRerouteUsage())
  FunctionOutput(None=LinearInterpolate(A=NamedRerouteUsage(), B=LinearInterpolate(A=@n6, B=@n7, Alpha=@n7) {constAlpha=1}, Alpha=NamedRerouteUsage())) {outputName=FinalOpacity, sortPriority=1}
  n8 = NamedRerouteUsage()
  n9 = MaterialFunctionCall!(UVs=NamedRerouteUsage(), Radius=NamedRerouteUsage()) {materialFunction=/FortUILibrary/MaterialFunctions/Shapes/SDFs/MF_SDF_Circle}
  NamedRerouteDeclaration(None=OneMinus(None=SmoothStep(Max=@n8, Value=@n9.SDF))) {name=CircleBackingAlpha, nodeColor=<rgba 0 0.8594 1 1>}
  n10 = NamedRerouteUsage()
  n11 = NamedRerouteUsage()
  n12 = TextureSample(UVs=NamedRerouteDeclaration(None=MaterialFunctionCall!(Texture Scale=@n10, UVs, Offset XY Texturesize Z=@n11) {materialFunction=/Engine/Functions/Engine_MaterialFunctions01/Texturing/ScaleUVsAroundPoint}.UVs) {name=UVs, nodeColor=<rgba 1 0.8203 0 1>}) {samplerSource=SSM_Clamp_WorldGroupSettings, samplerType=SAMPLERTYPE_Alpha, texture=/FortUILibrary/Textures/Icons/T_UI_Icon_UIKit_Locked}
  NamedRerouteDeclaration(None=@n12.A) {name=LockAlpha, nodeColor=<rgba 1 0.8203 0 1>}
  NamedRerouteDeclaration(None=FunctionInput(Preview=Constant() {r=0.1}) {bUsePreviewValueAsDefault=true, inputName=CircleBackingSoftness, inputType=FunctionInput_Scalar, sortPriority=4}) {name=CircleBackingSoftness, nodeColor=<rgba 0 1 0.0458 1>}
  NamedRerouteDeclaration(None=FunctionInput(Preview=Constant() {r=1.15}) {bUsePreviewValueAsDefault=true, inputName=CircleBackingRadius, inputType=FunctionInput_Scalar, sortPriority=5}) {name=CircleBackingRadius, nodeColor=<rgba 0 1 0.0458 1>}
  NamedRerouteDeclaration(None=Divide(A=NamedRerouteUsage(), B=Custom() {code="return View.ViewportScaleUI;", description=ViewportScaleFactor, outputType=CMOT_Float1, inputs []})) {name=DPIScaledSoftness, nodeColor=<rgba 0 0.9766 1 1>}
  NamedRerouteDeclaration(None=FunctionInput(Preview=Constant() {r=0.9}) {bUsePreviewValueAsDefault=true, inputName=CircleBackingOpacity, inputType=FunctionInput_Scalar, sortPriority=3}) {name=CircleBackingOpacity, nodeColor=<rgba 0 1 0.0458 1>}
  NamedRerouteDeclaration(None=FunctionInput(Preview=Constant() {r=0.8}) {bUsePreviewValueAsDefault=true, inputName=ImageLockedOpacity, inputType=FunctionInput_Scalar, sortPriority=6}) {name=ImageLockedOpacity, nodeColor=<rgba 0 1 0.0458 1>}
  n13 = FunctionInput(Preview=TextureSample() {texture=/QuestUI/QuestScreen/Textures/T_UI_QuestIcon_CinematicReward_L}.RGBA) {bUsePreviewValueAsDefault=true, inputName=InputV4, inputType=FunctionInput_Vector4}
  NamedRerouteDeclaration(None=ComponentMask(None=@n13) {b=true, g=true, r=true}) {name=RGBInput, nodeColor=<rgba 1 0.5839 0.0355 1>}
  NamedRerouteDeclaration(None=ComponentMask(None=@n13) {a=true}) {name=AlphaInput, nodeColor=<rgba 0.0078 1 0 1>}
  NamedRerouteDeclaration(None=FunctionInput(Preview=Constant() {r=1}) {bUsePreviewValueAsDefault=true, inputName=IsLockedCrossfade, inputType=FunctionInput_Scalar, sortPriority=1}) {name=IsLockedCrossfade, nodeColor=<rgba 0 1 0.0458 1>}
  NamedRerouteDeclaration(None=FunctionInput(Preview=Constant() {r=0.15}) {bUsePreviewValueAsDefault=true, inputName=LockScale, inputType=FunctionInput_Scalar, sortPriority=2}) {name=LockScale, nodeColor=<rgba 0.0078 1 0 1>}
  NamedRerouteDeclaration(None=FunctionInput(Preview=Constant3Vector() {constant=<rgba 0.9 0.2 1 0>}) {bUsePreviewValueAsDefault=true, inputName=Anchor, sortPriority=7}) {name=LockAnchor, nodeColor=<rgba 1 0.5839 0.0355 1>}
  n14 = NamedRerouteUsage()
  n15 = Subtract(A=@n9.SDF, B=NamedRerouteUsage())
  NamedRerouteDeclaration(None=OneMinus(None=SmoothStep(Max=@n14, Value=@n15) {constMax=2})) {name=CircleBackingInsetAlpha, nodeColor=<rgba 0 0.8594 1 1>}
 ]
}
