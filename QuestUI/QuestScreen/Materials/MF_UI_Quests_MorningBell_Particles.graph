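# UI material function that generates a panning field of procedural box particles, named for the Morning Bell quest screen.
# Panned UVs are tiled by a column count, and a seeded random value per tile drives horizontal offset, sine sway and pulsing size for MF_UI_Box.
# Screen-space derivatives of the tiled UVs mask seam artifacts in the fill.
# It outputs the masked grid, random values and unmasked SDF.
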MaterialFunction /QuestUI/QuestScreen/Materials/MF_UI_Quests_MorningBell_Particles
graph {
 properties {previewBlendMode=BLEND_Translucent, previewMaterialDomain=MD_UI}
 nodes [
  n1 = NamedRerouteUsage()
  n2 = FunctionInput() {bUsePreviewValueAsDefault=true, inputName=Pan Speed, inputType=FunctionInput_Vector2, sortPriority=2, previewValue=<0 0.5 0 1>}
  n3 = Multiply(A=Frac(None=Panner(Coordinate=@n1, Speed=@n2) {speedY=0.5}), B=NamedRerouteUsage()) {constB=4}
  n4 = Frac(None=@n3)
  n5 = Max(A=Abs(None=DDX(Value=@n4)), B=Abs(None=DDY(Value=@n4)))
  n6 = Max(A=ComponentMask(None=@n5) {r=true}, B=ComponentMask(None=@n5) {g=true})
  MaterialExpressionFunctionInput_4 = FunctionInput() {
   bUsePreviewValueAsDefault=true
   inputName=Horizontal Panning Range
   inputType=FunctionInput_Vector2
   sortPriority=2
   previewValue=<-0.1 0.1 0 0>
  }
  n7 = Convert(None=@MaterialExpressionFunctionInput_4) {
   nodeName=Break Float2
   convertInputs [{
     type=Vector2
     expressionInput {
      expression=&self:MaterialExpressionFunctionInput_4
      inputName=None
      mask=0
      maskA=0
      maskB=0
      maskG=0
      maskR=0
      outputIndex=0
     }
    }]
   convertOutputs [{type=Scalar}, {type=Scalar}]
  }
  n8 = LinearInterpolate(A=@n7, B=@n7, Alpha=NamedRerouteUsage()) {constA=-0.1, constB=0.1}
  n9 = AppendVector(A=@n8, B=@n8)
  n10 = Time()
  n11 = Multiply(A=NamedRerouteUsage()) {constB=0.5}
  n12 = Multiply(A=Sine(None=Multiply(A=@n10, B=@n11)), B=Multiply(A=NamedRerouteUsage()) {constB=0.1}) {constB=0.1}
  n13 = Constant()
  n14 = NamedRerouteUsage()
  n15 = FunctionInput() {bUsePreviewValueAsDefault=true, inputName=Sizing Variation Speed, inputType=FunctionInput_Scalar, sortPriority=2, previewValue=<0.25 0 0 1>}
  n16 = Multiply(A=Time(), B=@n15) {constB=0.25}
  MaterialExpressionFunctionInput_7 = FunctionInput() {
   bUsePreviewValueAsDefault=true
   inputName=Sizing Variation Range
   inputType=FunctionInput_Vector2
   sortPriority=2
   previewValue=<0.1 0.4 0 1>
  }
  n17 = Convert(None=@MaterialExpressionFunctionInput_7) {
   nodeName=Break Float2
   convertInputs [{
     type=Vector2
     expressionInput {
      expression=&self:MaterialExpressionFunctionInput_7
      inputName=None
      mask=0
      maskA=0
      maskB=0
      maskG=0
      maskR=0
      outputIndex=0
     }
    }]
   convertOutputs [{type=Scalar}, {type=Scalar}]
  }
  n18 = MaterialFunctionCall(
   UVs=Add(A=Add(A=@n4, B=@n9), B=AppendVector(A=@n12, B=@n13))
   Size=MaterialFunctionCall!(Sine Phase=Add(A=@n14, B=@n16), Value 1=@n17, Value 2=@n17) {materialFunction=/Engine/Functions/Engine_MaterialFunctions02/Utility/Sine_Remapped}.Remapped Sine
   GlowMax=Constant() {r=0.1}
  ) {materialFunction=&MF_UI_Box}
  FunctionOutput(None=Multiply(A=OneMinus(None=@n6), B=@n18.Fill) {constA=1}) {outputName=Grid}
  FunctionOutput(None=NamedRerouteUsage()) {outputName=Random Values, sortPriority=2}
  FunctionOutput(None=@n18.SDF) {outputName=Grid SDF, sortPriority=1}
  n19 = FunctionInput() {bUsePreviewValueAsDefault=true, inputName=Columns, inputType=FunctionInput_Scalar, sortPriority=1, previewValue=<4 0 0 1>}
  NamedRerouteDeclaration(None=@n19) {name=Columns, nodeColor=<rgba 0.4766 0 1 1>}
  NamedRerouteDeclaration(None=FunctionInput(Preview=TextureCoordinate()) {bUsePreviewValueAsDefault=true, inputName=UV, inputType=FunctionInput_Vector2}) {name=UV, nodeColor=<rgba 1 0 0.0234 1>}
  n20 = FunctionInput() {bUsePreviewValueAsDefault=true, inputName=Seed, inputType=FunctionInput_Scalar, sortPriority=2, previewValue=<0.3 0 0 1>}
  NamedRerouteDeclaration(None=MaterialFunctionCall!(UV=Floor(None=@n3), Seed=@n20) {materialFunction=&MF_UI_Random}.Result) {name=RandomTiles, nodeColor=<rgba 0 0 0 1>}
 ]
}
